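# Material function for the SteedShimmer Shire female body cosmetic that adds a cycling color gradient to a material's emissive.
# Repeated UVs pan vertically over scaled time, and the fractional V coordinate samples a color curve atlas row.
# Named reroutes carry this gradient and the Mask input to lerps that override both EmissiveColor and BaseColor on the incoming material attributes.
# Scalar parameters control gradient repetition, time scale and pan speed.
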MaterialFunction /BRCosmetics/Characters/Player/Female/Medium/Bodies/F_MED_SteedShimmer_Shire/Materials/MF_SteedShimmerCyclingRGBEmissive
graph {nodes [
  n1 = NamedRerouteUsage()
  n2 = NamedRerouteUsage()
  n3 = GetMaterialAttributes(None=FunctionInput() {inputName=MA, inputType=FunctionInput_MaterialAttributes}) {attributeGetTypes [#1, #2]}
  n4 = NamedRerouteUsage()
  n5 = NamedRerouteUsage()
  FunctionOutput(None=SetMaterialAttributes!(
    MaterialAttributes=@n3.MaterialAttributes
    EmissiveColor=LinearInterpolate(A=@n3.EmissiveColor, B=@n1, Alpha=@n2) {constB=0}
    BaseColor=LinearInterpolate(A=@n3.BaseColor, B=@n4, Alpha=@n5)
   ) {attributeSetTypes [#1, #2]})
  n6 = Panner(
   Coordinate=Multiply(A=TextureCoordinate(), B=ScalarParameter() {defaultValue=5, group=EmissivePannerGradient, parameterName=GradientRepetition}) {constB=5}
   Time=Multiply(A=Time(), B=ScalarParameter() {defaultValue=0.1, group=EmissivePannerGradient, parameterName=RGBGradientTimeScale})
   Speed=AppendVector(A=Constant(), B=ScalarParameter() {defaultValue=0.1, group=EmissivePannerGradient, parameterName=PanningGradientSpeed})
  ) {speedY=1}
  n7 = CurveAtlasRowParameter(CurveTime=ComponentMask(None=Frac(None=@n6)) {g=true}) {
   atlas=/BRCosmetics/Characters/Player/Female/Medium/Bodies/F_MED_SteedShimmer_Shire/Textures/CA_SteedShimmerEmissive_Gradient
   curve=/BRCosmetics/Characters/Player/Female/Medium/Bodies/F_MED_SteedShimmer_Shire/Textures/CC_SteedShimmer_ColorCurve
   group=EmissivePannerGradient
   parameterName=Emissive Gradient
  }
  NamedRerouteDeclaration(None=@n7) {name=Gradient, nodeColor=<rgba 0.6875 1 0 1>}
  NamedRerouteDeclaration(None=FunctionInput() {inputName=Mask, inputType=FunctionInput_Scalar}) {name=Mask, nodeColor=<rgba 0.9922 0 1 1>}
 ]}
