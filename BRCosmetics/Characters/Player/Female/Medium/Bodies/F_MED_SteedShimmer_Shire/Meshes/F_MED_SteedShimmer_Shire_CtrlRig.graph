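# Control Rig for the female SteedShimmer Shire body that adds procedural posing of the front and back middle cloth on top of a shared deform rig.
# Setup imports the skeleton and creates FK and thigh-twist controls; execution runs FK, deform rig, then a sequence.
# Cloth bones are parent-constrained to thighs or pelvis, then offset by spherical pose readers, with multiplied weights correcting two-leg poses.
# Collar, leg-swing, grind-rail and debug-print chains are left unconnected.
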ControlRigBlueprint /BRCosmetics/Characters/Player/Female/Medium/Bodies/F_MED_SteedShimmer_Shire/Meshes/F_MED_SteedShimmer_Shire_CtrlRig
graphs [
 New Function_ContainedGraph {nodes [
   n1 = RigVMFunctionEntryNode()
   RigVMFunction_MathQuaternionToEuler = &RigVMFunction_MathQuaternionToEuler(
    RotationOrder=XYZ
    Value=&RigUnit_GetRelativeTransformForItem(
     Child=(Type=Bone,Name="hand_l")
     Parent=(Type=Null,Name="lowerarm_l")
     bChildInitial=False
     bParentInitial=False
     Child.Name=@n1.Child
     Parent.Name=@n1.Parent
    ).RelativeTransform.Rotation
   )
   Remap = &RigVMFunction_MathDoubleRemap(
    TargetMaximum=1
    TargetMinimum=0
    Value=@RigVMFunction_MathQuaternionToEuler.Result.Y
    SourceMaximum=@n1.Source Maximum
    SourceMinimum=@n1.Source Minimum
   )
   > RigVMFunctionReturnNode(Delta Rotation=@Remap.Result)
  ]}
 RigVMFunctionLibrary {nodes [Delta Rotation Y = RigVMCollapseNode(ExecuteContext=())]}
 RigVMFunction_Sequence_1_ContainedGraph {nodes [
   n1 = RigVMFunctionEntryNode()
   RigVMFunction_Sequence = &RigVMFunction_Sequence(ExecuteContext=@n1.ExecuteContext)
   n2 = &RigVMFunction_Sequence(ExecuteContext=@RigVMFunction_Sequence.B)
   n3 = &RigVMFunction_Sequence(ExecuteContext=@n2.B)
   RigVMFunctionReturnNode(A=@RigVMFunction_Sequence.A, B=@n2.A, C=@n3.A, D=@n3.B)
  ]}
 RigVMModel {nodes [
   &RigUnit_ModifyTransforms(
    ExecutePin=()
    ItemToModify=((Item=(Type=Bone,Name="base_sweater_collar_01"),Transform=(Rotation=(X=0,Y=0,Z=0,W=1),Translation=(X=-3,Y=0,Z=0),Scale3D=(X=1,Y=1,Z=1))))
    Weight=1
   )
   > &RigUnit_ModifyTransforms(
    ItemToModify=((Item=(Type=Bone,Name="base_sweater_collar_02"),Transform=(Rotation=(X=0,Y=0,Z=0.4226,W=0.9063),Translation=(X=-4,Y=0,Z=0),Scale3D=(X=1,Y=1,Z=1))))
    Weight=1
   )
   > &RigUnit_ModifyTransforms(
    ItemToModify=((Item=(Type=Bone,Name="base_sweater_collar_03"),Transform=(Rotation=(X=0,Y=0,Z=0,W=1),Translation=(X=-1,Y=0,Z=0),Scale3D=(X=1,Y=1,Z=1))))
    Weight=1
   )
   > &RigUnit_ModifyTransforms(
    ItemToModify=((Item=(Type=Bone,Name="base_sweater_collar_06"),Transform=(Rotation=(X=0,Y=0.342,Z=0,W=0.9397),Translation=(X=-4,Y=0,Z=1),Scale3D=(X=1,Y=1,Z=1))))
    Weight=1
   )
   > &RigUnit_ModifyTransforms(
    ItemToModify=((Item=(Type=Bone,Name="base_sweater_collar_04"),Transform=(Rotation=(X=0,Y=0,Z=0.4226,W=0.9063),Translation=(X=-5,Y=-1,Z=0),Scale3D=(X=1,Y=1,Z=1))))
    Weight=1
   )
   n1 = RigVMVariableNode() {variable=leg_rot_z_minimum}
   Maximum = &RigVMFunction_MathDoubleMax(
    B=0
    A=&RigVMFunction_MathDoubleRemap(SourceMaximum=-180, SourceMinimum=0, TargetMaximum=15, TargetMinimum=0, Value=@n1.Value).Result
   )
   n2 = &RigVMFunction_MathDoubleRemap(SourceMaximum=-140, SourceMinimum=0, TargetMaximum=-80, TargetMinimum=0, Value=RigVMVariableNode() {variable=Leg_r_Swing_rot}.Value.Z)
   n3 = &RigVMFunction_MathDoubleRemap(SourceMaximum=-140, SourceMinimum=0, TargetMaximum=80, TargetMinimum=0, Value=RigVMVariableNode() {variable=Leg_l_Swing_rot}.Value.Z)
   n4 = &RigVMFunction_AnimEvalRichCurve(
    Curve="(EditorCurveData=(Keys=((ArriveTangent=1.1332),(InterpMode=RCIM_Cubic,TangentMode=RCTM_User,Time=0.1056,Value=0.2282,ArriveTangent=1.8977,LeaveTangent=1.8977),(InterpMode=RCIM_Cubic,TangentMode=RCTM_User,Time=0.2757,Value=0.3599,ArriveTangent=1.0257,LeaveT ...(truncated)"
    SourceMaximum=-230
    SourceMinimum=0
    TargetMaximum=-200
    TargetMinimum=0
    Value=RigVMVariableNode() {variable=leg_rot_z_minimum}.Value
   )
   &RigUnit_ModifyTransforms(
    ExecutePin=()
    ItemToModify=((Item=(Type=Bone,Name="base_frontMiddleCloth"),Transform=(Rotation=(X=0,Y=0,Z=0,W=1),Translation=(X=0,Y=0,Z=0),Scale3D=(X=1,Y=1,Z=1))))
    Weight=1
    WeightMinimum=0
    ItemToModify.0.Transform.Translation.Y=@Maximum.Result
    ItemToModify.0.Transform.Translation.X=/Script/RigVM.RigVMFunction_MathDoubleMin(
     B=0
     A=&RigVMFunction_MathDoubleRemap(SourceMaximum=-180, SourceMinimum=0, TargetMaximum=-15, TargetMinimum=0, Value=@n1.Value).Result
    ) {is=struct RigVMUnitNode, template="Minimum::Execute(in A,in B,out Result)"}.Result
    ItemToModify.0.Transform.Rotation=&RigVMFunction_MathQuaternionFromEuler(Euler=(X=0,Y=0,Z=0), Euler.Y=&RigVMFunction_MathDoubleAdd(A=@n2.Result, B=@n3.Result).Result, Euler.Z=@n4.Result).Result
   )
   n5 = &RigVMFunction_MathDoubleRemap(SourceMaximum=-140, SourceMinimum=0, TargetMaximum=-80, TargetMinimum=0, Value=RigVMVariableNode() {variable=Leg_r_Swing_rot}.Value.Z)
   n6 = &RigVMFunction_MathDoubleRemap(SourceMaximum=-140, SourceMinimum=0, TargetMaximum=80, TargetMinimum=0, Value=RigVMVariableNode() {variable=Leg_l_Swing_rot}.Value.Z)
   n7 = &RigVMFunction_MathDoubleRemap(SourceMaximum=120, SourceMinimum=0, TargetMaximum=7, TargetMinimum=0, Value=RigVMVariableNode() {variable=leg_rot_y_add}.Value)
   n8 = &RigVMFunction_MathQuaternionFromEuler(
    Euler=(X=0,Y=0,Z=0)
    RotationOrder=XYZ
    Euler.Z=&RigVMFunction_MathDoubleMul(A=&RigVMFunction_MathDoubleAdd(A=@n5.Result, B=@n6.Result).Result, B=@n7.Result).Result
    Euler.Y=&RigVMFunction_MathDoubleMul(B=16, A=@n7.Result).Result
   )
   > &RigUnit_ModifyTransforms(
    ItemToModify="((Item=(Type=Bone,Name=\"base_frontMiddleCloth\"),Transform=(Rotation=(X=0,Y=0,Z=-0.2588,W=0.9659),Translation=(X=0,Y=0,Z=0),Scale3D=(X=1,Y=1,Z=1))),(Item=(Type=Bone,Name=\"dyn_frontMiddleCloth_00\"),Transform=(Rotation=(X=0,Y=0,Z=0,W=1),Translation=(X=0,Y=0,Z ...(truncated)"
    Weight=1
    WeightMinimum=0
    ItemToModify.0.Transform.Rotation=@n8.Result
    ItemToModify.1.Transform.Rotation=@n8.Result
    ItemToModify.2.Transform.Rotation=@n8.Result
   )
   > n9 = &RigUnit_SphericalPoseReader(
    ActiveRegionScaleFactors=(PositiveWidth=1,NegativeWidth=1,PositiveHeight=1,NegativeHeight=1)
    ActiveRegionSize=0.28
    Debug=(bDrawDebug=false,bDraw2D=false,bDrawLocalAxes=false,DebugScale=25,DebugSegments=20,DebugThickness=0.25)
    DriverAxis=(X=1,Y=0,Z=0)
    DriverItem=(Type=Bone,Name="pelvis")
    FalloffRegionScaleFactors=(PositiveWidth=1,NegativeWidth=1,PositiveHeight=1,NegativeHeight=1)
    FalloffSize=0.33
    FlipHeightScaling=False
    FlipWidthScaling=False
    OptionalParentItem=(Type=Bone,Name="None")
    RotationOffset=(X=140,Y=45,Z=0)
   )
   > &RigUnit_ModifyTransforms(
    ItemToModify="((Item=(Type=Bone,Name=\"base_frontMiddleCloth\"),Transform=(Rotation=(X=-0.0922,Y=-0.1102,Z=-0.2638,W=0.9538),Translation=(X=0,Y=0,Z=0),Scale3D=(X=1,Y=1,Z=1))),(Item=(Type=Bone,Name=\"dyn_frontMiddleCloth_00\"),Transform=(Rotation=(X=0,Y=0,Z=0,W=1),Translatio ...(truncated)"
    WeightMinimum=0
    Weight=@n9.OutputParam
   )
   n10 = &RigUnit_PrepareForExecution()
   &RigUnit_HierarchyImportFromSkeleton(NameSpace=None, bIncludeCurves=False, bIncludeMeshSockets=False, bIncludeVirtualBones=True, ExecutePin=@n10.ExecutePin)
   > addFKCtrl = RigVMFunctionReferenceNode(
    Custom_FKCtrl=()
    Options_FKCtrl="(pelvis_1_#1=true,spine_3_#2=true,clavicle_5_#3=true,arm_7_#4=true,leg_9_#5=true,head_11_46DC15EF465CD2979 ...(truncated)"
    Shape_Settings="(pelvis_2_#6=(bVisible=True,Name=\"Hexagon_Thin\",Color=(R=0.0117,G=0.8281,B=1,A=1),Transform=(Rotation=(X=0,Y=-0.7071,Z=0,W=0.7071),Translation=(X=2,Y=0,Z=0),Scale3D=(X=5,Y=5,Z=7))),spine_4_#7=(bVi ...(truncated)"
   )
   > Control Stack at Item = RigVMFunctionReferenceNode(
    Add Bottom Null=False
    Add Null=False
    Add Secondary Control=False
    Control Suffix=_ctrl
    Item=(Name="thigh_twist_01_l",Type=Bone)
    Null Suffix=_null
    Parent=(Name="pelvis",Type=Bone)
    Secondary Settings="(Shape=(bVisible=True,Name=\"Default\",Color=(R=1,G=0,B=0,A=1),Transform=(Rotation=(X=0,Y=0,Z=0,W=1),Translation=(X=0,Y=0,Z=0),Scale3D=(X=1,Y=1,Z=1))),Proxy=(bIsProxy=False,ShapeVisibility=BasedOnSelection),DisplayName=\"None\",InitialSpace=LocalSpace,bUsePref ...(truncated)"
    Settings="(Shape=(bVisible=True,Name=\"Default\",Color=(R=0,G=0,B=1,A=1),Transform=(Rotation=(X=0,Y=0,Z=0,W=1),Translation=(X=0,Y=0,Z=0),Scale3D=(X=1,Y=1,Z=1))),Proxy=(bIsProxy=False,ShapeVisibility=BasedOnSelection),DisplayName=\"None\",InitialSpace=LocalSpace,bUsePref ...(truncated)"
    Tag Controls=true
    Zero Controls=true
   )
   > Control Stack at Item_1 = RigVMFunctionReferenceNode(
    Add Bottom Null=False
    Add Null=False
    Add Secondary Control=False
    Control Suffix=_ctrl
    Item=(Name="thigh_twist_01_r",Type=Bone)
    Null Suffix=_null
    Parent=(Name="pelvis",Type=Bone)
    Secondary Settings="(Shape=(bVisible=True,Name=\"Default\",Color=(R=1,G=0,B=0,A=1),Transform=(Rotation=(X=0,Y=0,Z=0,W=1),Translation=(X=0,Y=0,Z=0),Scale3D=(X=1,Y=1,Z=1))),Proxy=(bIsProxy=False,ShapeVisibility=BasedOnSelection),DisplayName=\"None\",InitialSpace=LocalSpace,bUsePref ...(truncated)"
    Settings="(Shape=(bVisible=True,Name=\"Default\",Color=(R=0,G=0,B=1,A=1),Transform=(Rotation=(X=0,Y=0,Z=0,W=1),Translation=(X=0,Y=0,Z=0),Scale3D=(X=1,Y=1,Z=1))),Proxy=(bIsProxy=False,ShapeVisibility=BasedOnSelection),DisplayName=\"None\",InitialSpace=LocalSpace,bUsePref ...(truncated)"
    Tag Controls=true
    Zero Controls=true
   )
   RigUnit_BeginExecution = &RigUnit_BeginExecution()
   setFKCtrl = RigVMFunctionReferenceNode(setFK_Ctrl=false, ExecuteContext=@RigUnit_BeginExecution.ExecutePin)
   > FN_DeformRig_FNC = RigVMFunctionReferenceNode(
    DeformRig_ Curve="(deformRig_l_upperarm_up_alpha_72_#8=False,deformRig_r_upperarm_up_alpha_73_#9=False,deformRig_l_upperarm_fr_alpha_74_#10=False,deformRig_r_upperarm_fr_alpha_75_CB51B8 ...(truncated)"
    DeformRig_Bool="(upperarm_1_#11=True,elbow_3_#12=True,wrist_7_#13=True,groin_8_#14=True,glute_17_#15=True,knee_9_40C408324CA261 ...(truncated)"
    DeformRig_Debug=(upperarm_debug_18_#11=False,wrist_debug_20_#13=False,lat_debug_26_#16=False,groinGlute_debug_27_#14=False)
    DeformRig_Value="(upperarm_up_18_#17=(Rotation=(X=0.0,Y=-0.0,Z=0.0,W=1.0),Translation=(X=0.0,Y=0.0,Z=8.0),Scale3D=(X=0.70,Y=1.0,Z=1.0)),upperarm_fwd_fr_19_#18=(Rotation=(X=0.0,Y=-0.0,Z=0.0,W=1.0),Translation=(X=0.0, ...(truncated)"
   )
   > RigVMFunction_Sequence = RigVMAggregateNode()
   &RigUnit_ParentConstraint(
    AdvancedSettings=(InterpolationType=Average,RotationOrderForFilter=XZY)
    Child=(Type=Bone,Name="base_frontMiddleCloth")
    Filter=(TranslationFilter=(bX=true,bY=true,bZ=true),RotationFilter=(bX=true,bY=true,bZ=true),ScaleFilter=(bX=True,bY=True,bZ=True))
    Parents=((Item=(Type=Bone,Name="thigh_l"),Weight=1),(Item=(Type=Bone,Name="thigh_r"),Weight=1))
    Weight=0.8
    ExecutePin=@RigVMFunction_Sequence.C
   )
   > n11 = &RigUnit_SphericalPoseReader(
    ActiveRegionScaleFactors=(PositiveWidth=3.5,NegativeWidth=0.5,PositiveHeight=3,NegativeHeight=3)
    ActiveRegionSize=0.1
    Debug=(bDrawDebug=false,bDraw2D=False,bDrawLocalAxes=False,DebugScale=25,DebugSegments=20,DebugThickness=0.25)
    DriverAxis=(X=1,Y=0,Z=0)
    DriverItem=(Type=Bone,Name="thigh_r")
    FalloffRegionScaleFactors=(PositiveWidth=1,NegativeWidth=1,PositiveHeight=1,NegativeHeight=1)
    FalloffSize=0.45
    FlipHeightScaling=False
    FlipWidthScaling=true
    OptionalParentItem=(Type=Bone,Name="pelvis")
    RotationOffset=(X=-90,Y=0,Z=0)
   )
   > &RigUnit_ModifyTransforms(
    ItemToModify="((Item=(Type=Bone,Name=\"base_frontMiddleCloth\"),Transform=(Rotation=(X=0.3146,Y=-0.1193,Z=-0.4633,W=0.8198),Translation=(X=-1,Y=4.8,Z=-1.8),Scale3D=(X=1,Y=1,Z=1))),(Item=(Type=Bone,Name=\"base_backMiddleCloth1\"),Transform=(Rotation=(X=0,Y=-0.7071,Z=0,W=0.70 ...(truncated)"
    Weight=@n11.OutputParam
   )
   > n12 = &RigUnit_SphericalPoseReader(
    ActiveRegionScaleFactors=(PositiveWidth=3.5,NegativeWidth=0.5,PositiveHeight=3,NegativeHeight=3)
    ActiveRegionSize=0.1
    Debug=(bDrawDebug=false,bDraw2D=False,bDrawLocalAxes=False,DebugScale=25,DebugSegments=20,DebugThickness=0.25)
    DriverAxis=(X=-1,Y=0,Z=0)
    DriverItem=(Type=Bone,Name="thigh_l")
    FalloffRegionScaleFactors=(PositiveWidth=1,NegativeWidth=1,PositiveHeight=1,NegativeHeight=1)
    FalloffSize=0.45
    FlipHeightScaling=False
    FlipWidthScaling=false
    OptionalParentItem=(Type=Bone,Name="pelvis")
    RotationOffset=(X=90,Y=0,Z=0)
   )
   > &RigUnit_ModifyTransforms(
    ItemToModify="((Item=(Type=Bone,Name=\"base_frontMiddleCloth\"),Transform=(Rotation=(X=-0.3055,Y=0.141,Z=-0.5194,W=0.7855),Translation=(X=-1,Y=4.8,Z=1.8),Scale3D=(X=1,Y=1,Z=1))),(Item=(Type=Bone,Name=\"base_backMiddleCloth1\"),Transform=(Rotation=(X=0,Y=0.7071,Z=0,W=0.7071) ...(truncated)"
    Weight=@n12.OutputParam
   )
   > &RigUnit_ModifyTransforms(
    ItemToModify="((Item=(Type=Bone,Name=\"base_frontMiddleCloth\"),Transform=(Rotation=(X=0.0438,Y=0.3132,Z=-0.5933,W=-0.7402),Translation=(X=7.5905,Y=-0.3192,Z=1.8258),Scale3D=(X=1,Y=1,Z=1))),(Item=(Type=Bone,Name=\"base_backMiddleCloth1\"),Transform=(Rotation=(X=0,Y=0,Z=-0.9 ...(truncated)"
    Weight=&RigVMFunction_MathDoubleMul(B=@n11.OutputParam, A=@n12.OutputParam).Result
   )
   &RigUnit_ParentConstraint(
    AdvancedSettings=(InterpolationType=Average,RotationOrderForFilter=XZY)
    Child=(Type=Bone,Name="base_backMiddleCloth1")
    Filter=(TranslationFilter=(bX=true,bY=true,bZ=true),RotationFilter=(bX=true,bY=true,bZ=true),ScaleFilter=(bX=True,bY=True,bZ=True))
    Parents=((Item=(Type=Bone,Name="thigh_l"),Weight=1),(Item=(Type=Bone,Name="thigh_r"),Weight=1),(Item=(Type=Bone,Name="pelvis"),Weight=1))
    Weight=0.8
    ExecutePin=@RigVMFunction_Sequence.D
   )
   > n13 = &RigUnit_SphericalPoseReader(
    ActiveRegionScaleFactors=(PositiveWidth=1,NegativeWidth=1,PositiveHeight=1,NegativeHeight=1)
    ActiveRegionSize=0.2
    Debug=(bDrawDebug=false,bDraw2D=false,bDrawLocalAxes=false,DebugScale=25,DebugSegments=20,DebugThickness=0.25)
    DriverAxis=(X=1,Y=0,Z=0)
    DriverItem=(Type=Bone,Name="thigh_l")
    FalloffRegionScaleFactors=(PositiveWidth=1,NegativeWidth=1,PositiveHeight=1,NegativeHeight=1)
    FalloffSize=0.3
    FlipHeightScaling=False
    FlipWidthScaling=False
    OptionalParentItem=(Type=Bone,Name="None")
    RotationOffset=(X=90,Y=0,Z=0)
   )
   > &RigUnit_ModifyTransforms(
    ItemToModify="((Item=(Type=Bone,Name=\"base_backMiddleCloth1\"),Transform=(Rotation=(X=0,Y=0,Z=0,W=1),Translation=(X=-4,Y=0,Z=2),Scale3D=(X=1,Y=1,Z=1))),(Item=(Type=Bone,Name=\"base_frontMiddleCloth\"),Transform=(Rotation=(X=0.0868,Y=-0.0076,Z=-0.0868,W=0.9924),Translation= ...(truncated)"
    Weight=@n13.OutputParam
   )
   > n14 = &RigUnit_SphericalPoseReader(
    ActiveRegionScaleFactors=(PositiveWidth=1,NegativeWidth=1,PositiveHeight=1,NegativeHeight=1)
    ActiveRegionSize=0.25
    Debug=(bDrawDebug=false,bDraw2D=false,bDrawLocalAxes=false,DebugScale=25,DebugSegments=20,DebugThickness=0.25)
    DriverAxis=(X=1,Y=0,Z=0)
    DriverItem=(Type=Bone,Name="thigh_r")
    FalloffRegionScaleFactors=(PositiveWidth=1,NegativeWidth=1,PositiveHeight=1,NegativeHeight=1)
    FalloffSize=0.35
    FlipHeightScaling=False
    FlipWidthScaling=False
    OptionalParentItem=(Type=Bone,Name="None")
    RotationOffset=(X=90,Y=0,Z=0)
   )
   > &RigUnit_ModifyTransforms(
    ItemToModify="((Item=(Type=Bone,Name=\"base_backMiddleCloth1\"),Transform=(Rotation=(X=-0.2113,Y=-0.0468,Z=0.2113,W=0.9532),Translation=(X=-4,Y=0,Z=-2),Scale3D=(X=1,Y=1,Z=1))),(Item=(Type=Bone,Name=\"base_frontMiddleCloth\"),Transform=(Rotation=(X=-0.0868,Y=0.0076,Z=-0.0868 ...(truncated)"
    Weight=@n14.OutputParam
   )
   > &RigUnit_ModifyTransforms(
    ItemToModify="((Item=(Type=Bone,Name=\"base_backMiddleCloth1\"),Transform=(Rotation=(X=0,Y=0,Z=0.0872,W=0.9962),Translation=(X=2,Y=-2,Z=0),Scale3D=(X=1,Y=1,Z=1))),(Item=(Type=Bone,Name=\"base_frontMiddleCloth\"),Transform=(Rotation=(X=0,Y=0,Z=0.1736,W=0.9848),Translation=(X ...(truncated)"
    Weight=&RigVMFunction_MathDoubleMul(B=@n13.OutputParam, A=@n14.OutputParam).Result
   )
   n15 = &RigUnit_SphericalPoseReader(
    ActiveRegionScaleFactors=(PositiveWidth=1,NegativeWidth=1,PositiveHeight=1,NegativeHeight=1)
    ActiveRegionSize=0.2
    Debug=(bDrawDebug=false,bDraw2D=false,bDrawLocalAxes=false,DebugScale=25,DebugSegments=20,DebugThickness=0.25)
    DriverAxis=(X=1,Y=0,Z=0)
    DriverItem=(Type=Bone,Name="thigh_l")
    ExecutePin=()
    FalloffRegionScaleFactors=(PositiveWidth=1,NegativeWidth=1,PositiveHeight=1,NegativeHeight=1)
    FalloffSize=0.3
    FlipHeightScaling=False
    FlipWidthScaling=False
    OptionalParentItem=(Type=Bone,Name="None")
    RotationOffset=(X=90,Y=0,Z=0)
   )
   > &RigUnit_ModifyTransforms(
    ItemToModify=((Item=(Type=Bone,Name="base_backMiddleCloth1"),Transform=(Rotation=(X=0.5032,Y=-0.3489,Z=-0.2604,W=0.7465),Translation=(X=0,Y=0,Z=10),Scale3D=(X=1,Y=1,Z=1))))
    Weight=@n15.OutputParam
   )
   > RigVMDispatchNode(Enabled=True, ScreenColor=(R=1,G=1,B=1,A=1), ScreenDuration=0.0, Value=@n15.OutputParam) {template="DISPATCH_RigVMDispatch_Print(in Prefix,in Value,in Enabled,in ScreenDuration,in ScreenColor)"}
   n16 = &RigUnit_SphericalPoseReader(
    ActiveRegionScaleFactors=(PositiveWidth=1,NegativeWidth=1,PositiveHeight=1,NegativeHeight=1)
    ActiveRegionSize=0.2
    Debug=(bDrawDebug=false,bDraw2D=false,bDrawLocalAxes=false,DebugScale=25,DebugSegments=20,DebugThickness=0.25)
    DriverAxis=(X=1,Y=0,Z=0)
    DriverItem=(Type=Bone,Name="thigh_r")
    ExecutePin=()
    FalloffRegionScaleFactors=(PositiveWidth=1,NegativeWidth=1,PositiveHeight=1,NegativeHeight=1)
    FalloffSize=0.3
    FlipHeightScaling=False
    FlipWidthScaling=False
    OptionalParentItem=(Type=Bone,Name="None")
    RotationOffset=(X=90,Y=0,Z=0)
   )
   > &RigUnit_ModifyTransforms(
    ItemToModify=((Item=(Type=Bone,Name="base_backMiddleCloth1"),Transform=(Rotation=(X=-0.7399,Y=-0.1983,Z=0.1664,W=0.6209),Translation=(X=-16,Y=0,Z=-7),Scale3D=(X=1,Y=1,Z=1))))
    Weight=@n16.OutputParam
   )
   > RigVMDispatchNode(Enabled=True, ScreenColor=(R=1,G=1,B=1,A=1), ScreenDuration=0.0, Value=@n16.OutputParam) {template="DISPATCH_RigVMDispatch_Print(in Prefix,in Value,in Enabled,in ScreenDuration,in ScreenColor)"}
   > n17 = &RigUnit_SphericalPoseReader(
    ActiveRegionScaleFactors=(PositiveWidth=1,NegativeWidth=1,PositiveHeight=1,NegativeHeight=1)
    ActiveRegionSize=0.2
    Debug=(bDrawDebug=false,bDraw2D=false,bDrawLocalAxes=false,DebugScale=25,DebugSegments=20,DebugThickness=0.25)
    DriverAxis=(X=1,Y=0,Z=0)
    DriverItem=(Type=Bone,Name="thigh_l")
    FalloffRegionScaleFactors=(PositiveWidth=1,NegativeWidth=1,PositiveHeight=1,NegativeHeight=1)
    FalloffSize=0.3
    FlipHeightScaling=False
    FlipWidthScaling=False
    OptionalParentItem=(Type=Bone,Name="None")
    RotationOffset=(X=270,Y=0,Z=0)
   )
   > &RigUnit_ModifyTransforms(
    ItemToModify=((Item=(Type=Bone,Name="base_backMiddleCloth1"),Transform=(Rotation=(X=-0.6088,Y=0,Z=0,W=0.7934),Translation=(X=0,Y=0,Z=0),Scale3D=(X=1,Y=1,Z=1))))
    Weight=@n17.OutputParam
   )
   > n18 = &RigUnit_SphericalPoseReader(
    ActiveRegionScaleFactors=(PositiveWidth=1,NegativeWidth=1,PositiveHeight=1,NegativeHeight=1)
    ActiveRegionSize=0.2
    Debug=(bDrawDebug=false,bDraw2D=false,bDrawLocalAxes=false,DebugScale=25,DebugSegments=20,DebugThickness=0.25)
    DriverAxis=(X=1,Y=0,Z=0)
    DriverItem=(Type=Bone,Name="thigh_r")
    FalloffRegionScaleFactors=(PositiveWidth=1,NegativeWidth=1,PositiveHeight=1,NegativeHeight=1)
    FalloffSize=0.3
    FlipHeightScaling=False
    FlipWidthScaling=False
    OptionalParentItem=(Type=Bone,Name="None")
    RotationOffset=(X=270,Y=0,Z=0)
   )
   > &RigUnit_ModifyTransforms(
    ItemToModify=((Item=(Type=Bone,Name="base_backMiddleCloth1"),Transform=(Rotation=(X=0.6088,Y=0,Z=0,W=0.7934),Translation=(X=0,Y=0,Z=0),Scale3D=(X=1,Y=1,Z=1))))
    Weight=@n18.OutputParam
   )
   > &RigUnit_ModifyTransforms(
    ItemToModify=((Item=(Type=Bone,Name="base_backMiddleCloth1"),Transform=(Rotation=(X=0,Y=0,Z=-0.866,W=0.5),Translation=(X=0,Y=0,Z=0),Scale3D=(X=1,Y=1,Z=1))))
    Weight=0
   )
   n19 = &RigUnit_SphericalPoseReader(
    ActiveRegionScaleFactors=(PositiveWidth=1,NegativeWidth=1,PositiveHeight=1,NegativeHeight=1)
    ActiveRegionSize=0.2
    Debug=(bDrawDebug=false,bDraw2D=false,bDrawLocalAxes=false,DebugScale=25,DebugSegments=20,DebugThickness=0.25)
    DriverAxis=(X=1,Y=0,Z=0)
    DriverItem=(Type=Bone,Name="thigh_r")
    ExecutePin=()
    FalloffRegionScaleFactors=(PositiveWidth=1,NegativeWidth=1,PositiveHeight=1,NegativeHeight=1)
    FalloffSize=0.5
    FlipHeightScaling=False
    FlipWidthScaling=False
    OptionalParentItem=(Type=Bone,Name="None")
    RotationOffset=(X=270,Y=0,Z=0)
   )
   > &RigUnit_ModifyTransforms(
    ItemToModify=((Item=(Type=Bone,Name="base_frontMiddleCloth"),Transform=(Rotation=(X=0.3851,Y=-0.2302,Z=-0.6686,W=0.593),Translation=(X=-2,Y=2,Z=0),Scale3D=(X=1,Y=1,Z=1))))
    Weight=1
   )
   > &RigUnit_ModifyTransforms(
    ItemToModify=((Item=(Type=Bone,Name="base_backMiddleCloth1"),Transform=(Rotation=(X=0,Y=0,Z=-0.342,W=0.9397),Translation=(X=0,Y=0,Z=0),Scale3D=(X=1,Y=1,Z=1))))
    Weight=1
   )
   Multiply = &RigVMFunction_MathDoubleMul(A=1, B=@n19.OutputParam)
   > &RigUnit_ModifyTransforms(
    ItemToModify=((Item=(Type=Bone,Name="base_frontMiddleCloth"),Transform=(Rotation=(X=0,Y=-0.0698,Z=0,W=0.9976),Translation=(X=0,Y=0,Z=0),Scale3D=(X=1,Y=1,Z=1))))
    Weight=@Multiply.Result
   )
   > &RigUnit_ModifyTransforms(
    ItemToModify=((Item=(Type=Bone,Name="base_backMiddleCloth1"),Transform=(Rotation=(X=0,Y=0,Z=0,W=1),Translation=(X=0,Y=0,Z=0),Scale3D=(X=1,Y=1,Z=1))))
    Weight=@Multiply.Result
   )
   n20 = &RigUnit_SphericalPoseReader(
    ActiveRegionScaleFactors=(PositiveWidth=3.5,NegativeWidth=0.5,PositiveHeight=4.2,NegativeHeight=3)
    ActiveRegionSize=0.1
    Debug=(bDrawDebug=false,DebugScale=25,DebugSegments=20,DebugThickness=0.25)
    DriverAxis=(X=-1,Y=0,Z=0)
    DriverItem=(Type=Bone,Name="thigh_l")
    ExecutePin=()
    FalloffRegionScaleFactors=(PositiveWidth=1,NegativeWidth=1.11,PositiveHeight=4,NegativeHeight=1)
    FalloffSize=0.45
    OptionalParentItem=(Type=Bone)
    RotationOffset=(X=90,Y=0,Z=0)
   )
   > RigVMVariableNode(Value=@n20.OutputParam) {variable=l_thigh_up_alpha}
   > n21 = &RigUnit_SphericalPoseReader(
    ActiveRegionScaleFactors=(PositiveWidth=3.5,NegativeWidth=0.5,PositiveHeight=4.2,NegativeHeight=3)
    ActiveRegionSize=0.1
    Debug=(bDrawDebug=false,DebugScale=25,DebugSegments=20,DebugThickness=0.25)
    DriverAxis=(X=1,Y=0,Z=0)
    DriverItem=(Type=Bone,Name="thigh_r")
    FalloffRegionScaleFactors=(PositiveWidth=1,NegativeWidth=1.11,PositiveHeight=4,NegativeHeight=1)
    FalloffSize=0.45
    FlipHeightScaling=false
    FlipWidthScaling=true
    OptionalParentItem=(Type=Bone)
    RotationOffset=(X=-90,Y=0,Z=0)
   )
   > RigVMVariableNode(Value=@n21.OutputParam) {variable=r_thigh_up_alpha}
   > n22 = &RigUnit_SphericalPoseReader(
    ActiveRegionScaleFactors=(PositiveWidth=3.5,NegativeWidth=0.5,PositiveHeight=3,NegativeHeight=3)
    ActiveRegionSize=0.1
    Debug=(bDrawDebug=false,bDraw2D=false,DebugScale=25,DebugSegments=20,DebugThickness=0.25)
    DriverAxis=(X=-1,Y=0,Z=0)
    DriverItem=(Type=Bone,Name="thigh_l")
    FalloffRegionScaleFactors=(PositiveWidth=1,NegativeWidth=1.11,PositiveHeight=1,NegativeHeight=1)
    FalloffSize=0.45
    OptionalParentItem=(Type=Bone)
    RotationOffset=(X=-90,Y=0,Z=0)
   )
   > RigVMVariableNode(Value=@n22.OutputParam) {variable=l_thigh_back_alpha}
   > n23 = &RigUnit_SphericalPoseReader(
    ActiveRegionScaleFactors=(PositiveWidth=3.5,NegativeWidth=0.5,PositiveHeight=3,NegativeHeight=3)
    ActiveRegionSize=0.1
    Debug=(bDrawDebug=false,DebugScale=25,DebugSegments=20,DebugThickness=0.25)
    DriverAxis=(X=1,Y=0,Z=0)
    DriverItem=(Type=Bone,Name="thigh_r")
    FalloffRegionScaleFactors=(PositiveWidth=1,NegativeWidth=1.11,PositiveHeight=1,NegativeHeight=1)
    FalloffSize=0.45
    FlipHeightScaling=false
    FlipWidthScaling=true
    OptionalParentItem=(Type=Bone)
    RotationOffset=(X=90,Y=0,Z=0)
   )
   > RigVMVariableNode(Value=@n23.OutputParam) {variable=r_thigh_back_alpha}
   AlphaInterp = &RigVMFunction_AlphaInterp(
    InterpSpeedDecreasing=10
    InterpSpeedIncreasing=10
    Scale=1
    bClampResult=true
    Value=RigVMDispatchNode(False=1, True=0.6, Condition=RigVMVariableNode() {variable=grindOnRail}.Value) {template="DISPATCH_RigVMDispatch_If(in Condition,in True,in False,out Result)"}.Result
   )
   > RigVMVariableNode(Value=@AlphaInterp.Result) {variable=grindReduction}
   n24 = &RigUnit_GetRelativeTransformForItem(Child=(Type=Bone,Name="thigh_twist_01_r"), Parent=(Type=Control,Name="thigh_twist_01_r_ctrl"), bChildInitial=False, bParentInitial=False)
   RigVMFunction_MathQuaternionSwingTwist = &RigVMFunction_MathQuaternionSwingTwist(TwistAxis=(X=1,Y=0,Z=0), Input=@n24.RelativeTransform.Rotation)
   RigVMFunction_MathQuaternionToEuler = &RigVMFunction_MathQuaternionToEuler(RotationOrder=ZYX, Value=@RigVMFunction_MathQuaternionSwingTwist.Swing)
   RigVMVariableNode(ExecuteContext=(), Value=@RigVMFunction_MathQuaternionToEuler.Result) {variable=Leg_r_Swing_rot}
   n25 = &RigUnit_GetRelativeTransformForItem(Child=(Type=Bone,Name="thigh_twist_01_l"), Parent=(Type=Control,Name="thigh_twist_01_l_ctrl"), bChildInitial=False, bParentInitial=False)
   n26 = &RigVMFunction_MathQuaternionToEuler(
    RotationOrder=ZYX
    Value=&RigVMFunction_MathQuaternionSwingTwist(TwistAxis=(X=1,Y=0,Z=0), Input=@n25.RelativeTransform.Rotation).Swing
   )
   > RigVMVariableNode(Value=@n26.Result) {variable=Leg_l_Swing_rot}
   > RigVMVariableNode(Value=&RigVMFunction_MathDoubleMin(A=@RigVMFunction_MathQuaternionToEuler.Result.Z, B=@n26.Result.Z) {template="Minimum::Execute(in A,in B,out Result)"}.Result) {variable=leg_rot_z_minimum}
   > RigVMVariableNode(Value=&RigVMFunction_MathDoubleAdd(A=@RigVMFunction_MathQuaternionToEuler.Result.Y, B=@n26.Result.Y).Result) {variable=leg_rot_y_add}
   > RigVMVariableNode(Value=&RigVMFunction_MathDoubleMax(A=@RigVMFunction_MathQuaternionToEuler.Result.Z, B=@n26.Result.Z).Result) {variable=leg_rot_z_maximum}
   &RigVMFunction_MathDoubleMax(A=0, B=0)
   &RigVMFunction_MathDoubleMin(
    B=&RigVMFunction_MathDoubleRemap(SourceMaximum=-100, SourceMinimum=0, TargetMaximum=-15, TargetMinimum=0, Value=RigVMVariableNode() {variable=leg_rot_y_add}.Value).Result
    A=&RigVMFunction_MathDoubleRemap(SourceMaximum=-180, SourceMinimum=0, TargetMaximum=-15, TargetMinimum=0, Value=RigVMVariableNode() {variable=leg_rot_z_minimum}.Value).Result
   ) {template="Minimum::Execute(in A,in B,out Result)"}
   n27 = &RigVMFunction_MathDoubleRemap(SourceMaximum=140, SourceMinimum=0, TargetMaximum=150, TargetMinimum=0, Value=RigVMVariableNode() {variable=Leg_r_Swing_rot}.Value.Z)
   n28 = &RigVMFunction_MathDoubleRemap(SourceMaximum=140, SourceMinimum=0, TargetMaximum=150, TargetMinimum=0, Value=RigVMVariableNode() {variable=Leg_l_Swing_rot}.Value.Z)
   &RigUnit_ModifyTransforms(
    ExecutePin=()
    ItemToModify="((Item=(Type=Bone,Name=\"dyn_backMiddleCloth_00\"),Transform=(Rotation=(X=-0.4226,Y=0,Z=0,W=0.9063),Translation=(X=0,Y=0,Z=0),Scale3D=(X=1,Y=1,Z=1))),(Item=(Type=Bone,Name=\"base_backMiddleCloth1\"),Transform=(Rotation=(X=0,Y=0,Z=0,W=1),Translation=(X=0,Y=0,Z= ...(truncated)"
    Weight=1
    WeightMinimum=0
    ItemToModify.0.Transform.Rotation=&RigVMFunction_MathQuaternionFromEuler(Euler=(X=0,Y=0,Z=0), RotationOrder=XYZ, Euler.Z=&RigVMFunction_MathDoubleMax(B=@n27.Result, A=@n28.Result).Result).Result
   )
   n29 = &RigVMFunction_MathDoubleRemap(SourceMaximum=-140, SourceMinimum=0, TargetMaximum=-40, TargetMinimum=0, Value=RigVMVariableNode() {variable=Leg_r_Swing_rot}.Value.Z)
   n30 = &RigVMFunction_MathDoubleRemap(SourceMaximum=-140, SourceMinimum=0, TargetMaximum=40, TargetMinimum=0, Value=RigVMVariableNode() {variable=Leg_l_Swing_rot}.Value.Z)
   n31 = &RigVMFunction_AnimEvalRichCurve(
    Curve="(EditorCurveData=(Keys=((InterpMode=RCIM_Linear,TangentMode=RCTM_Auto,TangentWeightMode=RCTWM_WeightedNone,Time=0,Value=0,ArriveTangent=1.1332,ArriveTangentWeight=0,LeaveTangent=0,LeaveTangentWeight=0),(InterpMode=RCIM_Linear,TangentMode=RCTM_Auto,TangentW ...(truncated)"
    SourceMaximum=-160
    SourceMinimum=0
    TargetMinimum=0
    TargetMaximum=&RigVMFunction_MathDoubleMul(B=-40, A=RigVMVariableNode() {variable=grindReduction}.Value).Result
    Value=RigVMVariableNode() {variable=leg_rot_z_minimum}.Value
   )
   &RigUnit_ModifyTransforms(
    ExecutePin=()
    ItemToModify=((Item=(Type=Bone,Name="base_backMiddleCloth1"),Transform=(Rotation=(X=0,Y=0,Z=0,W=1),Translation=(X=0,Y=0,Z=0),Scale3D=(X=1,Y=1,Z=1))))
    Weight=1
    WeightMinimum=0
    ItemToModify.0.Transform.Rotation=&RigVMFunction_MathQuaternionFromEuler(Euler=(X=0,Y=30,Z=0), Euler.Y=&RigVMFunction_MathDoubleAdd(A=@n29.Result, B=@n30.Result).Result, Euler.Z=@n31.Result).Result
   )
   &RigVMFunction_MathDoubleMul(B=-160, A=RigVMVariableNode() {variable=grindReduction}.Value)
   &RigVMFunction_MathDoubleMul(B=@n18.OutputParam, A=@n17.OutputParam)
   RigVMDispatchNode(Enabled=True, ExecuteContext=(), Prefix=R, ScreenColor=(R=1,G=1,B=1,A=1), ScreenDuration=0.0, Value=@n11.OutputParam) {template="DISPATCH_RigVMDispatch_Print(in Prefix,in Value,in Enabled,in ScreenDuration,in ScreenColor)"}
   RigVMDispatchNode(Enabled=True, ExecuteContext=(), Prefix=L, ScreenColor=(R=1,G=1,B=1,A=1), ScreenDuration=0.0, Value=@n12.OutputParam) {template="DISPATCH_RigVMDispatch_Print(in Prefix,in Value,in Enabled,in ScreenDuration,in ScreenColor)"}
   RigVMDispatchNode(Enabled=True, ExecuteContext=(), Prefix=back_LEG, ScreenColor=(R=1,G=1,B=1,A=1), ScreenDuration=0.0, Value=@n14.OutputParam) {template="DISPATCH_RigVMDispatch_Print(in Prefix,in Value,in Enabled,in ScreenDuration,in ScreenColor)"}
   RigVMDispatchNode(Enabled=True, ExecuteContext=(), Prefix=leg, ScreenColor=(R=1,G=1,B=1,A=1), ScreenDuration=0.0, Value=@n13.OutputParam) {template="DISPATCH_RigVMDispatch_Print(in Prefix,in Value,in Enabled,in ScreenDuration,in ScreenColor)"}
  ]}
]
variables [
 GetCrouchingAlpha {type=double, default=0}
 Leg_l_Swing_rot {type=FVector, default=(X=0,Y=0,Z=0)}
 Leg_r_Swing_rot {type=FVector, default=(X=0,Y=0,Z=0)}
 SetLoinclothFront {type=FTransform, default=(Rotation=(X=0,Y=0,Z=0,W=1),Translation=(X=0,Y=0,Z=0),Scale3D=(X=1,Y=1,Z=1))}
 Sliding {type=bool, default=False}
 grindOnRail {type=bool, default=False}
 grindReduction {type=double, default=0}
 isRunningOrEmote {type=bool, default=False}
 l_thigh_back_alpha {type=double, default=0}
 l_thigh_up_alpha {type=double, default=0}
 leg_rot_y_add {type=double, default=0}
 leg_rot_z_maximum {type=double, default=0}
 leg_rot_z_minimum {type=double, default=0}
 r_thigh_back_alpha {type=double, default=0}
 r_thigh_up_alpha {type=double, default=0}
]
