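# Anim Blueprint for the SteedShimmer Shire female body, adding secondary motion to sleeves, boots and front cloth.
# A copied parent pose runs through Control Rig and AnimDynamics into a cached pose.
# A bool blends two boot-dynamics strengths, then bone offsets and a calf look-at adjust the front cloth.
# The event graph updates gravity override by movement state, drives cloth-sim material wind parameters from skydiving and flags specific emote montages.
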AnimBlueprint /BRCosmetics/Characters/Player/Female/Medium/Bodies/F_MED_SteedShimmer_Shire/Meshes/F_MED_SteedShimmer_Shire_AnimBP
target_skeleton=&/F_MED_SteedShimmer_Shire_Skeleton
graphs [
 AnimGraph {nodes [
   n1 = Anim Dynamics - Bone: base_acc_boot_01_lf!(ComponentPose=Local To Component(LocalPose=Use cached pose 'animdynamics_pose'()), GravityOverride:Vector="0,0,-980", ExternalForce="0,0,0") -> (Pose) {
    alpha=0.2
    angularDampingOverride=0.9999
    angularSpringConstant=8000
    bOverrideAngularDamping=true
    bUseGravityOverride=true
    bUsePlanarLimit=false
    gravityScale=0.35
    simulationSpace=RootRelative
    worldSpaceMinimumScale=0
    alphaBoolBlend {blendOption=Linear}
    boundBone {boneName=base_acc_boot_01_lf}
    componentAppliedLinearAccClamp=<0 0 0>
    gravityOverride {z=-980}
    physicsBodyDefinitions [{
      boundBone {boneName=base_acc_boot_01_lf}
      boxExtents=<5 5 5>
      constraintSetup {angularLimitsMax=<0 0 45>, angularLimitsMin=<0 0 -45>, angularTarget=<1 0 0>}
      localJointOffset=<5 0 0>
     }]
    relativeSpaceBone {boneName=base_controller}
    simSpaceSettings {simSpaceAngularAlpha=1}
   }
   n2 = Anim Dynamics - Bone: base_acc_boot_01_rt!(ComponentPose=@n1, GravityOverride:Vector="0,0,-980", ExternalForce="0,0,0") -> (Pose) {
    alpha=0.2
    angularDampingOverride=0.9999
    angularSpringConstant=8000
    bOverrideAngularDamping=true
    bUseGravityOverride=true
    bUsePlanarLimit=false
    gravityScale=0.35
    simulationSpace=RootRelative
    worldSpaceMinimumScale=0
    alphaBoolBlend {blendOption=Linear}
    boundBone {boneName=base_acc_boot_01_rt}
    componentAppliedLinearAccClamp=<0 0 0>
    gravityOverride {z=-980}
    physicsBodyDefinitions [{
      boundBone {boneName=base_acc_boot_01_rt}
      boxExtents=<5 5 5>
      constraintSetup {angularLimitsMax=<0 0 45>, angularLimitsMin=<0 0 -45>, angularTarget=<1 0 0>}
      localJointOffset=<5 0 0>
     }]
    relativeSpaceBone {boneName=base_controller}
    simSpaceSettings {simSpaceAngularAlpha=1}
   }
   n3 = Component To Local(ComponentPose=@n2)
   n4 = Anim Dynamics - Bone: base_acc_boot_01_lf(ComponentPose=Local To Component(LocalPose=Use cached pose 'animdynamics_pose'()), GravityOverride="0,0,-980", ExternalForce="0,0,0") {
    angularDampingOverride=0.9999
    angularSpringConstant=8000
    bOverrideAngularDamping=true
    bUseGravityOverride=true
    bUsePlanarLimit=false
    gravityScale=0.35
    simulationSpace=RootRelative
    worldSpaceMinimumScale=0
    alphaBoolBlend {blendOption=Linear}
    boundBone {boneName=base_acc_boot_01_lf}
    componentAppliedLinearAccClamp=<0 0 0>
    gravityOverride {z=-980}
    physicsBodyDefinitions [{
      boundBone {boneName=base_acc_boot_01_lf}
      boxExtents=<5 5 5>
      constraintSetup {angularLimitsMax=<0 0 45>, angularLimitsMin=<0 0 -45>, angularTarget=<1 0 0>}
      localJointOffset=<5 0 0>
     }]
    relativeSpaceBone {boneName=base_controller}
    simSpaceSettings {simSpaceAngularAlpha=1}
   }
   n5 = Anim Dynamics - Bone: base_acc_boot_01_rt(ComponentPose=@n4, GravityOverride="0,0,-980", ExternalForce="0,0,0") {
    angularDampingOverride=0.9999
    angularSpringConstant=8000
    bOverrideAngularDamping=true
    bUseGravityOverride=true
    bUsePlanarLimit=false
    gravityScale=0.35
    simulationSpace=RootRelative
    worldSpaceMinimumScale=0
    alphaBoolBlend {blendOption=Linear}
    boundBone {boneName=base_acc_boot_01_rt}
    componentAppliedLinearAccClamp=<0 0 0>
    gravityOverride {z=-980}
    physicsBodyDefinitions [{
      boundBone {boneName=base_acc_boot_01_rt}
      boxExtents=<5 5 5>
      constraintSetup {angularLimitsMax=<0 0 45>, angularLimitsMin=<0 0 -45>, angularTarget=<1 0 0>}
      localJointOffset=<5 0 0>
     }]
    relativeSpaceBone {boneName=base_controller}
    simSpaceSettings {simSpaceAngularAlpha=1}
   }
   n6 = Component To Local(ComponentPose=@n5)
   n7 = "Transform (Modify) Bone - Bone: base_frontMiddleCloth"!(
    ComponentPose=Local To Component(LocalPose=Blend Poses by bool(BlendPose_0=@n3, BlendPose_1=@n6))
    bAlphaBoolEnabled:bool=True
    Translation="0,-1,0"
    Rotation="15,0,0"
    Scale="1,1,1"
   ) -> (Pose) {
    alphaInputType=Bool
    rotationMode=BMM_Additive
    rotationSpace=BCS_BoneSpace
    translationMode=BMM_Additive
    translationSpace=BCS_BoneSpace
    alphaBoolBlend {blendInTime=1, blendOption=Cubic, blendOutTime=5}
    boneToModify {boneName=base_frontMiddleCloth}
    rotation {pitch=15}
    translation {y=-1}
   }
   n8 = "Transform (Modify) Bone - Bone: dyn_frontMiddleCloth_01"!(ComponentPose=@n7, bAlphaBoolEnabled:bool=True, Translation="0,0,0", Rotation="10,0,0", Scale="1,1,1") -> (Pose) {
    alphaInputType=Bool
    rotationMode=BMM_Additive
    rotationSpace=BCS_BoneSpace
    alphaBoolBlend {blendInTime=1, blendOption=Cubic, blendOutTime=5}
    boneToModify {boneName=dyn_frontMiddleCloth_01}
    rotation {pitch=10}
   }
   n9 = "Transform (Modify) Bone - Bone: dyn_frontMiddleCloth_02"!(ComponentPose=@n8, bAlphaBoolEnabled:bool=True, Translation="0,0,0", Rotation="5,0,0", Scale="1,1,1") -> (Pose) {
    alphaInputType=Bool
    rotationMode=BMM_Additive
    rotationSpace=BCS_BoneSpace
    alphaBoolBlend {blendInTime=1, blendOption=Cubic, blendOutTime=5}
    boneToModify {boneName=dyn_frontMiddleCloth_02}
    rotation {pitch=5}
   }
   n10 = "Transform (Modify) Bone - Bone: base_frontMiddleCloth"(ComponentPose=@n9, bAlphaBoolEnabled=True, Translation="0,0,1", Rotation="-20,-5,0", Scale="1,1,1") {
    alphaInputType=Bool
    rotationMode=BMM_Additive
    rotationSpace=BCS_BoneSpace
    translationMode=BMM_Additive
    translationSpace=BCS_BoneSpace
    alphaBoolBlend {blendInTime=1, blendOption=Cubic, blendOutTime=5}
    boneToModify {boneName=base_frontMiddleCloth}
    rotation {pitch=-20, yaw=-5}
    translation {z=1}
   }
   n11 = "Transform (Modify) Bone - Bone: base_frontMiddleCloth"(ComponentPose=@n10, bAlphaBoolEnabled=True, Translation="0,0,1", Rotation="-5,-7.5,0", Scale="1,1,1") {
    alphaInputType=Bool
    rotationMode=BMM_Additive
    rotationSpace=BCS_BoneSpace
    translationMode=BMM_Additive
    translationSpace=BCS_BoneSpace
    alphaBoolBlend {blendInTime=1, blendOption=Cubic, blendOutTime=5}
    boneToModify {boneName=base_frontMiddleCloth}
    rotation {pitch=-5, yaw=-7.5}
    translation {z=1}
   }
   n12 = Look At - Bone: base_frontMiddleCloth(ComponentPose=@n11, Alpha=1) {
    alphaScaleBiasClamp {bInterpResult=true}
    boneToModify {boneName=base_frontMiddleCloth}
    lookAtLocation=<-3.7028 -28.1633 -6.1597>
    lookAtTarget {boneReference {boneName=calf_r}}
    lookAt_Axis {axis=<1 0>}
   }
   Output Pose(Result=Component To Local(ComponentPose=@n12)) {layerGroup=None}
   n13 = Control Rig(Source=Copy Pose From Mesh() {bUseAttachedParent=true}) {controlRigAssetReference {blueprintRigClass=&/F_MED_SteedShimmer_Shire_CtrlRig.F_MED_SteedShimmer_Shire_CtrlRig_C}}
   n14 = Anim Dynamics - Bone: dyn_sleeves_01_lf!(ComponentPose=Local To Component(LocalPose=@n13), GravityOverride:Vector="0,0,-980", ExternalForce="0,0,0") -> (Pose) {
    alpha=0.8
    angularDampingOverride=0.9999
    angularSpringConstant=8000
    bUseGravityOverride=true
    bUsePlanarLimit=false
    gravityScale=0.35
    simulationSpace=RootRelative
    worldSpaceMinimumScale=0
    alphaBoolBlend {blendOption=Linear}
    boundBone {boneName=dyn_sleeves_01_lf}
    componentAppliedLinearAccClamp=<0 0 0>
    gravityOverride {z=-980}
    physicsBodyDefinitions [{
      boundBone {boneName=dyn_sleeves_01_lf}
      boxExtents=<20 15 10>
      constraintSetup {angularLimitsMax=<0 5 0>, angularLimitsMin=<0 -5 0>, angularTarget=<1 0 0>, linearAxesMax=<0 0 1.5>, linearAxesMin=<0 0 -1.5>}
      localJointOffset=<0 0 0>
     }]
   }
   n15 = Anim Dynamics - Bone: dyn_sleeves_02_lf!(ComponentPose=@n14, GravityOverride:Vector="0,0,-980", ExternalForce="0,0,0") -> (Pose) {
    alpha=0.8
    angularDampingOverride=0.9999
    angularSpringConstant=8000
    bUseGravityOverride=true
    bUsePlanarLimit=false
    gravityScale=0.35
    simulationSpace=RootRelative
    worldSpaceMinimumScale=0
    alphaBoolBlend {blendOption=Linear}
    boundBone {boneName=dyn_sleeves_02_lf}
    componentAppliedLinearAccClamp=<0 0 0>
    gravityOverride {z=-980}
    physicsBodyDefinitions [{
      boundBone {boneName=dyn_sleeves_02_lf}
      boxExtents=<20 15 10>
      constraintSetup {angularLimitsMax=<0 5 0>, angularLimitsMin=<0 -5 0>, angularTarget=<1 0 0>, linearAxesMax=<0 0 1.5>, linearAxesMin=<0 0 -1.5>}
      localJointOffset=<0 0 0>
     }]
   }
   n16 = Anim Dynamics - Bone: dyn_sleeves_03_lf!(ComponentPose=@n15, GravityOverride:Vector="0,0,-980", ExternalForce="0,0,0") -> (Pose) {
    alpha=0.8
    angularDampingOverride=0.9999
    angularSpringConstant=8000
    bUseGravityOverride=true
    bUsePlanarLimit=false
    gravityScale=0.35
    simulationSpace=RootRelative
    worldSpaceMinimumScale=0
    alphaBoolBlend {blendOption=Linear}
    boundBone {boneName=dyn_sleeves_03_lf}
    componentAppliedLinearAccClamp=<0 0 0>
    gravityOverride {z=-980}
    physicsBodyDefinitions [{
      boundBone {boneName=dyn_sleeves_03_lf}
      boxExtents=<20 15 10>
      constraintSetup {angularLimitsMax=<0 5 0>, angularLimitsMin=<0 -5 0>, angularTarget=<1 0 0>, linearAxesMax=<0 0 1.5>, linearAxesMin=<0 0 -1.5>}
      localJointOffset=<0 0 0>
     }]
   }
   n17 = Anim Dynamics - Bone: dyn_sleeves_04_lf!(ComponentPose=@n16, GravityOverride:Vector="0,0,-980", ExternalForce="0,0,0") -> (Pose) {
    alpha=0.8
    angularDampingOverride=0.9999
    angularSpringConstant=8000
    bUseGravityOverride=true
    bUsePlanarLimit=false
    gravityScale=0.35
    simulationSpace=RootRelative
    worldSpaceMinimumScale=0
    alphaBoolBlend {blendOption=Linear}
    boundBone {boneName=dyn_sleeves_04_lf}
    componentAppliedLinearAccClamp=<0 0 0>
    gravityOverride {z=-980}
    physicsBodyDefinitions [{
      boundBone {boneName=dyn_sleeves_04_lf}
      boxExtents=<20 15 10>
      constraintSetup {angularLimitsMax=<0 5 0>, angularLimitsMin=<0 -5 0>, angularTarget=<1 0 0>, linearAxesMax=<0 0 1.5>, linearAxesMin=<0 0 -1.5>}
      localJointOffset=<0 0 0>
     }]
   }
   n18 = Anim Dynamics - Bone: dyn_sleeves_05_lf!(ComponentPose=@n17, GravityOverride:Vector="0,0,-980", ExternalForce="0,0,0") -> (Pose) {
    alpha=0.8
    angularDampingOverride=0.9999
    angularSpringConstant=8000
    bUseGravityOverride=true
    bUsePlanarLimit=false
    gravityScale=0.35
    simulationSpace=RootRelative
    worldSpaceMinimumScale=0
    alphaBoolBlend {blendOption=Linear}
    boundBone {boneName=dyn_sleeves_05_lf}
    componentAppliedLinearAccClamp=<0 0 0>
    gravityOverride {z=-980}
    physicsBodyDefinitions [{
      boundBone {boneName=dyn_sleeves_05_lf}
      boxExtents=<20 15 10>
      constraintSetup {angularLimitsMax=<0 5 0>, angularLimitsMin=<0 -5 0>, angularTarget=<1 0 0>, linearAxesMax=<0 0 0.5>, linearAxesMin=<0 0 -0.8>}
      localJointOffset=<0 0 0>
     }]
   }
   n19 = Anim Dynamics - Bone: dyn_sleeves_01_rt!(ComponentPose=@n18, GravityOverride:Vector="0,0,-980", ExternalForce="0,0,0") -> (Pose) {
    alpha=0.8
    angularDampingOverride=0.9999
    angularSpringConstant=8000
    bUseGravityOverride=true
    bUsePlanarLimit=false
    gravityScale=0.35
    simulationSpace=RootRelative
    worldSpaceMinimumScale=0
    alphaBoolBlend {blendOption=Linear}
    boundBone {boneName=dyn_sleeves_01_rt}
    componentAppliedLinearAccClamp=<0 0 0>
    gravityOverride {z=-980}
    physicsBodyDefinitions [{
      boundBone {boneName=dyn_sleeves_01_rt}
      boxExtents=<20 15 10>
      constraintSetup {angularLimitsMax=<0 5 0>, angularLimitsMin=<0 -5 0>, angularTarget=<1 0 0>, linearAxesMax=<0 0 1.5>, linearAxesMin=<0 0 -1.5>}
      localJointOffset=<0 0 0>
     }]
   }
   n20 = Anim Dynamics - Bone: dyn_sleeves_02_rt!(ComponentPose=@n19, GravityOverride:Vector="0,0,-980", ExternalForce="0,0,0") -> (Pose) {
    alpha=0.8
    angularDampingOverride=0.9999
    angularSpringConstant=8000
    bUseGravityOverride=true
    bUsePlanarLimit=false
    gravityScale=0.35
    simulationSpace=RootRelative
    worldSpaceMinimumScale=0
    alphaBoolBlend {blendOption=Linear}
    boundBone {boneName=dyn_sleeves_02_rt}
    componentAppliedLinearAccClamp=<0 0 0>
    gravityOverride {z=-980}
    physicsBodyDefinitions [{
      boundBone {boneName=dyn_sleeves_02_rt}
      boxExtents=<20 15 10>
      constraintSetup {angularLimitsMax=<0 5 0>, angularLimitsMin=<0 -5 0>, angularTarget=<1 0 0>, linearAxesMax=<0 0 1.5>, linearAxesMin=<0 0 -1.5>}
      localJointOffset=<0 0 0>
     }]
   }
   n21 = Anim Dynamics - Bone: dyn_sleeves_03_rt!(ComponentPose=@n20, GravityOverride:Vector="0,0,-980", ExternalForce="0,0,0") -> (Pose) {
    alpha=0.8
    angularDampingOverride=0.9999
    angularSpringConstant=8000
    bUseGravityOverride=true
    bUsePlanarLimit=false
    gravityScale=0.35
    simulationSpace=RootRelative
    worldSpaceMinimumScale=0
    alphaBoolBlend {blendOption=Linear}
    boundBone {boneName=dyn_sleeves_03_rt}
    componentAppliedLinearAccClamp=<0 0 0>
    gravityOverride {z=-980}
    physicsBodyDefinitions [{
      boundBone {boneName=dyn_sleeves_03_rt}
      boxExtents=<20 15 10>
      constraintSetup {angularLimitsMax=<0 5 0>, angularLimitsMin=<0 -5 0>, angularTarget=<1 0 0>, linearAxesMax=<0 0 1.5>, linearAxesMin=<0 0 -1.5>}
      localJointOffset=<0 0 0>
     }]
   }
   n22 = Anim Dynamics - Bone: dyn_sleeves_04_rt!(ComponentPose=@n21, GravityOverride:Vector="0,0,-980", ExternalForce="0,0,0") -> (Pose) {
    alpha=0.8
    angularDampingOverride=0.9999
    angularSpringConstant=8000
    bUseGravityOverride=true
    bUsePlanarLimit=false
    gravityScale=0.35
    simulationSpace=RootRelative
    worldSpaceMinimumScale=0
    alphaBoolBlend {blendOption=Linear}
    boundBone {boneName=dyn_sleeves_04_rt}
    componentAppliedLinearAccClamp=<0 0 0>
    gravityOverride {z=-980}
    physicsBodyDefinitions [{
      boundBone {boneName=dyn_sleeves_04_rt}
      boxExtents=<20 15 10>
      constraintSetup {angularLimitsMax=<0 5 0>, angularLimitsMin=<0 -5 0>, angularTarget=<1 0 0>, linearAxesMax=<0 0 1.5>, linearAxesMin=<0 0 -1.5>}
      localJointOffset=<0 0 0>
     }]
   }
   n23 = Anim Dynamics - Bone: dyn_sleeves_05_rt!(ComponentPose=@n22, GravityOverride:Vector="0,0,-980", ExternalForce="0,0,0") -> (Pose) {
    alpha=0.8
    angularDampingOverride=0.9999
    angularSpringConstant=8000
    bUseGravityOverride=true
    bUsePlanarLimit=false
    gravityScale=0.35
    simulationSpace=RootRelative
    worldSpaceMinimumScale=0
    alphaBoolBlend {blendOption=Linear}
    boundBone {boneName=dyn_sleeves_05_rt}
    componentAppliedLinearAccClamp=<0 0 0>
    gravityOverride {z=-980}
    physicsBodyDefinitions [{
      boundBone {boneName=dyn_sleeves_05_rt}
      boxExtents=<20 15 10>
      constraintSetup {angularLimitsMax=<0 5 0>, angularLimitsMin=<0 -5 0>, angularTarget=<1 0 0>, linearAxesMax=<0 0 0.5>, linearAxesMin=<0 0 -0.8>}
      localJointOffset=<0 0 0>
     }]
   }
   n24 = Anim Dynamics - Bone: dyn_controller!(ComponentPose=@n23, GravityOverride:Vector="0,0,-980", ExternalForce="0,0,0") -> (Pose) {
    angularDampingOverride=0.9999
    angularSpringConstant=8000
    bUseGravityOverride=true
    bUsePlanarLimit=false
    gravityScale=0.35
    simulationSpace=RootRelative
    worldSpaceMinimumScale=1
    alphaBoolBlend {blendOption=Linear}
    boundBone {boneName=dyn_controller}
    componentAppliedLinearAccClamp=<0 0 0>
    gravityOverride {z=-980}
    physicsBodyDefinitions [{
      boundBone {boneName=dyn_controller}
      boxExtents=<20 15 10>
      constraintSetup {angularLimitsMax=<0 0 7>, angularLimitsMin=<0 0 -7>, angularTarget=<1 0 0>}
      localJointOffset=<0 0 0>
     }]
    relativeSpaceBone {boneName=base_controller}
    simSpaceSettings {simSpaceAngularAlpha=1}
   }
   Save cached pose 'animdynamics_pose'(Pose=Component To Local(ComponentPose=@n24)) {cacheName=animdynamics_pose}
  ]}
 EventGraph {nodes [
   Event BlueprintUpdateAnimation()
   > n1 = Sequence!(execute) -> (then_0, then_1, then_2:Exec)
   > n2 = "Gravity Override State Function (Nativized)"!(
    execute
    CharacterPartAnimInstance:Custom Character Part Anim Instance Object Reference=Self-Reference()
    GravityOverrideParameterSet:Fort Gravity Override Parameter Set Structure=Get Gravity Override Params() -> (Gravity Override Params:Fort Gravity Override Parameter Set Structure)
    RigidBodyAnimNodeParameterSet:Fort Rigid Body Anim Node Parameter Set Structure=Get RBAN Params() -> (RBAN Params:Fort Rigid Body Anim Node Parameter Set Structure)
    DebugDraw:bool=false
    GameTimeInSeconds:float=GetGameTimeInSeconds()
    DebugJointName:Name=root
    StateLogicLayerName:Name=none
   ) -> (
    then
    OutGravityOverride:Vector
    OutCurrentState:Name=None
    OutGravityOverrideParameters:Fort Gravity Override Parameters Structure
    OutRigidBodyAnimNodeParameters:Fort Rigid Body Anim Node Parameters Structure
   )
   > Set Gravity Override(Gravity Override=@n2)
   > n3 = Branch(Condition=Get bIsSkyDiving())
   > Set WPO Pho Wind Intensity(WPO Pho Wind Intensity=1)
   > Set WPO Pho Turbulence(WPO Pho Turbulence:double=0.5) -> (Output_Get:double)
   > n4 = Set WPO Pho Drag(WPO Pho Drag:double=1) -> (Output_Get:double)
   SetScalarParameterValue(
    execute=[@n4.then, @n7.then]
    self="Get Mat Instance Dynamic "() -> ("Mat Instance Dynamic ":Material Instance Dynamic Object Reference)
    ParameterName=ClothSim_WindStrength
    Value=Get WPO Pho Wind Intensity()
   )
   n5 = Get WPO Pho Turbulence() -> (WPO Pho Turbulence:double)
   > SetScalarParameterValue(self="Get Mat Instance Dynamic "(), ParameterName=ClothSim_R1_TurbulenceStrength, Value=@n5)
   > SetScalarParameterValue(self="Get Mat Instance Dynamic "(), ParameterName=ClothSim_R2_TurbulenceStrength, Value=@n5)
   n6 = Get WPO Pho Drag() -> (WPO Pho Drag:double)
   > SetScalarParameterValue(self="Get Mat Instance Dynamic "(), ParameterName=ClothSim_R1_DragStrength, Value=@n6)
   > SetScalarParameterValue(self="Get Mat Instance Dynamic "(), ParameterName=ClothSim_R2_DragStrength, Value=@n6)
   Set WPO Pho Wind Intensity(execute=@n3.else, WPO Pho Wind Intensity=0)
   > Set WPO Pho Turbulence(WPO Pho Turbulence=0.2)
   > n7 = Set WPO Pho Drag(WPO Pho Drag=5)
   n8 = GetAnimInstance(self=GetOwnerSkeletalMeshComponent())
   Is Valid(exec=@n1.then_1, InputObject=@n8)
   n9 = Montage_IsActive!(
    self:Anim Instance Object Reference=@n8
    Montage:Anim Montage Object Reference=/BRCosmetics/Animation/Game/MainPlayer/Emotes/RowLiaison/Emote_RowLiaison_CMF_M
   ) -> (ReturnValue:bool)
   > Set BudBlossomEmote(BudBlossomEmote:bool=OR Boolean(A=Montage_IsActive(self=@n8, Montage=/BRCosmetics/Animation/Game/MainPlayer/Emotes/BudBlossom/CMF/Emote_BudBlossom_CMF_M), B=@n9)) -> (Output_Get:bool)
   > Set RowLiaisonEmote(RowLiaisonEmote:bool=@n9) -> (Output_Get:bool)
   n10 = GetAnimInstance(self=GetOwnerSkeletalMeshComponent())
   n11 = Montage_IsPlaying(self=@n10, Montage=/BRCosmetics/Animation/Game/MainPlayer/Emotes/GoldenGlory/CMF/Emote_GoldenGlory_CMF_M)
   n12 = Montage_IsPlaying(self=@n10, Montage=/BRCosmetics/Animation/Game/MainPlayer/Emotes/GoldenGlory/CMF/Emote_GoldenGlory_CMF_FrontEnd_M)
   Set GoldenGloryEmote(execute=@n1.then_2, GoldenGloryEmote:double="To Float (Boolean)"(InBool=OR Boolean(A=@n11, B=@n12))) -> (Output_Get:double)
   Event OnCharacterCustomizationLoaded()
   > n13 = Set SkelMesh(SkelMesh=GetPartSkeletalMeshComponent())
   > n14 = Get Dynamic Material Instance(SkeletalMeshComponent=@n13)
   > "Set Mat Instance Dynamic "("Mat Instance Dynamic ":Material Instance Dynamic Object Reference=@n14) -> (Output_Get:Material Instance Dynamic Object Reference)
   > n15 = GetOrCreateDynamicOverlayMaterial(MeshComponent=@n13, MaterialName=None)
   > "Set Overlya Mat Instance Dynamic "("Overlya Mat Instance Dynamic ":Material Instance Dynamic Object Reference=@n15) -> (Output_Get:Material Instance Dynamic Object Reference)
   Event BlueprintInitializeAnimation()
   > Set SkelMesh(SkelMesh=GetOwnerSkeletalMeshComponent())
   n16 = TryGetPawnOwner()
   > Is Valid(InputObject=@n16)
   > n17 = Cast To FortPlayerPawnAthena(Object=@n16)
   > Set PlayerPawn(PlayerPawn:Fort Player Pawn Athena Object Reference=@n17) -> (Output_Get:Fort Player Pawn Athena Object Reference)
  ]}
]
variables [
 BudBlossomEmote {type=boolean}
 GoldenGloryEmote {type=number}
 Gravity Override {type=Vector}
 Gravity Override Params {
  type=FortGravityOverrideParameterSet
  default=(Skydive_GoingUp=(GravityOverride=()),Skydive_FloatingDown=(GravityOverride=()),Skydive_DivingDown=(GravityOverride=()),Skydive_Parachuting=(GravityOverride=()),OnGround_Standing=(GravityOverride=()),OnGround_Moving=(GravityOverride=()),OnGround_Moving_HighRotationRate=(GravityOverride=()),OnGround_Crouch_Moving=(GravityOverride=()),EmoteOrMelee=(GravityOverride=()),Swimming=(GravityOverride=()),Front_End=(GravityOverride=()),DBNO=(GravityOverride=()),DBNO_Carried=(GravityOverride=()),OnGround_Falling=(GravityOverride=()),Ziplining=(GravityOverride=()),Skydive_Surfing=(GravityOverride=()),Riding_Standing=(GravityOverride=()),Riding_Moving=(GravityOverride=()),Riding_Moving_Targeting=(GravityOverride=()),Tactical_Sprint=(GravityOverride=()),Rail_Grinding=(GravityOverride=()),)
 }
 "Mat Instance Dynamic " {type=MaterialInstanceDynamic}
 "Overlya Mat Instance Dynamic " {type=MaterialInstanceDynamic}
 PlayerPawn {type=FortPlayerPawnAthena}
 RBAN Params {
  type=FortRigidBodyAnimNodeParameterSet
  default=(Skydive_GoingUp=(RigidBodyAnimNode=()),Skydive_FloatingDown=(RigidBodyAnimNode=()),Skydive_DivingDown=(RigidBodyAnimNode=()),Skydive_Parachuting=(RigidBodyAnimNode=()),OnGround_Standing=(RigidBodyAnimNode=()),OnGround_Moving=(RigidBodyAnimNode=()),OnGround_Moving_HighRotationRate=(RigidBodyAnimNode=()),OnGround_Crouch_Moving=(RigidBodyAnimNode=()),EmoteOrMelee=(RigidBodyAnimNode=()),Swimming=(RigidBodyAnimNode=()),Front_End=(RigidBodyAnimNode=()),DBNO=(RigidBodyAnimNode=()),DBNO_Carried=(RigidBodyAnimNode=()),OnGround_Falling=(RigidBodyAnimNode=()),Ziplining=(RigidBodyAnimNode=()),Skydive_Surfing=(RigidBodyAnimNode=()),Riding_Standing=(RigidBodyAnimNode=()),Riding_Moving=(RigidBodyAnimNode=()),Riding_Moving_Targeting=(RigidBodyAnimNode=()),Tactical_Sprint=(RigidBodyAnimNode=()),Rail_Grinding=(RigidBodyAnimNode=()),)
 }
 RowLiaisonEmote {type=boolean}
 SkelMesh {type=SkeletalMeshComponent}
 WPO Pho Drag {type=number}
 WPO Pho Turbulence {type=number}
 WPO Pho Wind Intensity {type=number}
]
