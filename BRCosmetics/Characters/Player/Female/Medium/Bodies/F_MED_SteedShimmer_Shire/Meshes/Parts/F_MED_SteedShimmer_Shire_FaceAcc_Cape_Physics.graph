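# Physics asset for the SteedShimmer Shire female character's FaceAcc_Cape mesh part, set up for secondary motion on bang and ponytail bone chains.
# Capsule bodies cover pelvis, spine, neck, head and arms as collision volumes, joined mostly by constraints with free angular motion.
# Branching ponytail chains hang from base_ponytail on the head, and forked bang chains from base_hair on neck_01.
# These chains use narrow swing limits with zero twist.
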PhysicsAsset /BRCosmetics/Characters/Player/Female/Medium/Bodies/F_MED_SteedShimmer_Shire/Meshes/Parts/F_MED_SteedShimmer_Shire_FaceAcc_Cape_Physics
bodies {
 base_hair [{shapeType=Capsule, length=0, radius=1}]
 base_ponytail [{shapeType=Capsule, length=0, radius=1, shapeName=base_ponytail_capsule, rotation=<rot 90 0 0>}]
 clavicle_l [{shapeType=Capsule, length=8, radius=4, shapeName=clavicle_l_capsule, center=<2 0 0>, rotation=<rot 90 0 0>}]
 clavicle_r [{shapeType=Capsule, length=8, radius=4, shapeName=clavicle_r_capsule, center=<-2 0 0>, rotation=<rot 90 180 0>}]
 dyn_bang_01 [{shapeType=Capsule, length=3, radius=1.5, shapeName=dyn_bang_01_capsule, center=<2 0 0>, rotation=<rot 90 0 0>}]
 dyn_bang_02 [{shapeType=Capsule, length=2.653, radius=1.5, shapeName=dyn_bang_02_capsule, center=<2 0 0>, rotation=<rot 90 0 0>}]
 dyn_bang_03 [{shapeType=Capsule, length=1, radius=1, shapeName=dyn_bang_03_capsule, center=<1 0 0>, rotation=<rot 90 0 0>}]
 dyn_bang_04 [{shapeType=Capsule, length=1, radius=1, shapeName=dyn_bang_04_capsule, center=<1 0 0>, rotation=<rot 90 0 0>}]
 dyn_bang_05 [{shapeType=Capsule, length=1.5, radius=1.15, shapeName=dyn_bang_05_capsule, center=<1 0 0>, rotation=<rot 90 0 0>}]
 dyn_bang_06 [{shapeType=Capsule, length=2, radius=0.75, shapeName=dyn_bang_06_capsule, center=<1 0 0>, rotation=<rot 90 0 0>}]
 dyn_bang_07 [{shapeType=Capsule, length=2, radius=0.75, shapeName=dyn_bang_07_capsule, center=<1 0 0>, rotation=<rot 90 0 0>}]
 dyn_ponytailM_01 [{shapeType=Capsule, length=3, radius=6, shapeName=dyn_ponytailM_01_capsule, center=<4 0 0>, rotation=<rot 90 0 0>}]
 dyn_ponytailM_03 [{shapeType=Capsule, length=3, radius=6, shapeName=dyn_ponytailM_03_capsule, center=<4 0 0>, rotation=<rot 90 0 0>}]
 dyn_ponytailM_04 [{shapeType=Capsule, length=3, radius=6, shapeName=dyn_ponytailM_04_capsule, center=<4 0 0>, rotation=<rot 90 0 0>}]
 dyn_ponytailM_05 [
  {shapeType=Capsule, length=3, radius=6, shapeName=dyn_ponytailM_05_capsule, center=<4 0 0>, rotation=<rot 90 0 0>}
  {shapeType=Capsule, length=4.9451, radius=5, shapeName=dyn_ponytailM_05_capsule, center=<7.6885 7 -1.8748>, rotation=<rot 90 0 0>}
  {shapeType=Capsule, length=3, radius=4, shapeName=dyn_ponytailM_05_capsule, center=<6 -7 -2>, rotation=<rot 90 0 0>}
 ]
 dyn_ponytailM_06 [
  {shapeType=Capsule, length=3.6994, radius=5.3006, shapeName=dyn_ponytailM_06_capsule, center=<4 0 0>, rotation=<rot 90 0 0>}
  {shapeType=Capsule, length=2.5114, radius=4.6616, shapeName=dyn_ponytailM_06_capsule, center=<6 4 0>, rotation=<rot 90 0 0>}
  {shapeType=Capsule, length=2.5, radius=4.6, shapeName=dyn_ponytailM_06_capsule, center=<6 -4 0>, rotation=<rot 90 0 0>}
 ]
 dyn_ponytailM_07 [{shapeType=Capsule, length=4.6582, radius=4.3418, shapeName=dyn_ponytailM_07_capsule, center=<4 0 0>, rotation=<rot 90 0 0>}]
 dyn_ponytailM_08 [{shapeType=Capsule, length=4.737, radius=4.263, shapeName=dyn_ponytailM_08_capsule, center=<4 0 0>, rotation=<rot 90 0 0>}]
 head [
  {shapeType=Capsule, length=0.5703, radius=8.3899, shapeName=head_capsule, center=<5.3315 -0.9535 0>, rotation=<rot 90 -16.7174 0>}
  {shapeType=Capsule, length=0, radius=4.6995, shapeName=head_capsule, center=<1.8091 0 -7.7218>}
  {shapeType=Capsule, length=0, radius=4.6995, shapeName=head_capsule, center=<1.8091 0 7.7218>}
  {shapeType=Capsule, length=5.5109, radius=5.3811, shapeName=head_capsule, center=<1.184 3.8147 0>, rotation=<rot 90 0 0>}
 ]
 lowerarm_l [{shapeType=Capsule, length=27.9184, radius=5.8728, shapeName=upperarm_l_capsule, center=<15 0 -1>, rotation=<rot 90 0 0>}]
 lowerarm_r [{shapeType=Capsule, length=27.9184, radius=5.8728, shapeName=upperarm_r_capsule, center=<-15 0 1>, rotation=<rot 90 180 0>}]
 neck_01 [{shapeType=Capsule, length=0, radius=5, shapeName=neck_01_capsule}]
 neck_02 [{shapeType=Capsule, length=0, radius=5, shapeName=neck_02_capsule}]
 pelvis [{shapeType=Capsule, length=7.2089, radius=14.6338, shapeName=spine_05_capsule, center=<-5.1042 0 0>}]
 spine_01 [{shapeType=Capsule, length=8.6305, radius=10, shapeName=spine_05_capsule}]
 spine_02 [{shapeType=Capsule, length=7.2416, radius=10, shapeName=spine_05_capsule}]
 spine_03 [{shapeType=Capsule, length=6.5915, radius=10, shapeName=spine_05_capsule}]
 spine_04 [
  {shapeType=Capsule, length=7.2817, radius=10, shapeName=spine_05_capsule}
  {shapeType=Capsule, length=2.3995, radius=15, shapeName=spine_04_capsule, center=<-5.2378 -6.7247 0>}
 ]
 spine_05 [
  {shapeType=Capsule, length=9.3616, radius=10, shapeName=spine_05_capsule, center=<0.1156 0.2288 0>}
  {shapeType=Capsule, length=2.3995, radius=15, shapeName=spine_05_capsule, center=<-1.3586 -5.6718 0>}
 ]
 upperarm_l [{shapeType=Capsule, length=27.9184, radius=5.8728, shapeName=upperarm_l_capsule, center=<15 0 -1>, rotation=<rot 90 0 0>}]
 upperarm_r [{shapeType=Capsule, length=27.9184, radius=5.8728, shapeName=upperarm_r_capsule, center=<-15 0 1>, rotation=<rot 90 180 0>}]
}
constraints [
 {bone1Name=base_hair, bone2Name=neck_01, swing1LimitDegrees=45, swing2LimitDegrees=45, twistLimitDegrees=45}
 {bone1Name=dyn_ponytailM_04, bone2Name=dyn_ponytailM_01, swing1LimitDegrees=35, swing2LimitDegrees=35, twistLimitDegrees=0}
 {bone1Name=dyn_ponytailM_06, bone2Name=dyn_ponytailM_04, swing1LimitDegrees=35, swing2LimitDegrees=35, twistLimitDegrees=0}
 {
  bone1Name=spine_01
  bone2Name=pelvis
  swing1LimitDegrees=45
  swing1Motion=Free
  swing2LimitDegrees=45
  swing2Motion=Free
  twistLimitDegrees=45
  twistMotion=Free
 }
 {
  bone1Name=spine_02
  bone2Name=spine_01
  swing1LimitDegrees=45
  swing1Motion=Free
  swing2LimitDegrees=45
  swing2Motion=Free
  twistLimitDegrees=45
  twistMotion=Free
 }
 {
  bone1Name=spine_03
  bone2Name=spine_02
  swing1LimitDegrees=45
  swing1Motion=Free
  swing2LimitDegrees=45
  swing2Motion=Free
  twistLimitDegrees=45
  twistMotion=Free
 }
 {
  bone1Name=spine_04
  bone2Name=spine_03
  swing1LimitDegrees=45
  swing1Motion=Free
  swing2LimitDegrees=45
  swing2Motion=Free
  twistLimitDegrees=45
  twistMotion=Free
 }
 {
  bone1Name=spine_05
  bone2Name=spine_04
  swing1LimitDegrees=45
  swing1Motion=Free
  swing2LimitDegrees=45
  swing2Motion=Free
  twistLimitDegrees=45
  twistMotion=Free
 }
 {
  bone1Name=clavicle_l
  bone2Name=spine_05
  swing1LimitDegrees=45
  swing1Motion=Free
  swing2LimitDegrees=45
  swing2Motion=Free
  twistLimitDegrees=45
  twistMotion=Free
 }
 {
  bone1Name=upperarm_l
  bone2Name=clavicle_l
  swing1LimitDegrees=45
  swing1Motion=Free
  swing2LimitDegrees=45
  swing2Motion=Free
  twistLimitDegrees=45
  twistMotion=Free
 }
 {
  bone1Name=lowerarm_l
  bone2Name=upperarm_l
  swing1LimitDegrees=45
  swing1Motion=Free
  swing2LimitDegrees=45
  swing2Motion=Free
  twistLimitDegrees=45
  twistMotion=Free
 }
 {
  bone1Name=clavicle_r
  bone2Name=spine_05
  swing1LimitDegrees=45
  swing1Motion=Free
  swing2LimitDegrees=45
  swing2Motion=Free
  twistLimitDegrees=45
  twistMotion=Free
 }
 {
  bone1Name=upperarm_r
  bone2Name=clavicle_r
  swing1LimitDegrees=45
  swing1Motion=Free
  swing2LimitDegrees=45
  swing2Motion=Free
  twistLimitDegrees=45
  twistMotion=Free
 }
 {
  bone1Name=lowerarm_r
  bone2Name=upperarm_r
  swing1LimitDegrees=45
  swing1Motion=Free
  swing2LimitDegrees=45
  swing2Motion=Free
  twistLimitDegrees=45
  twistMotion=Free
 }
 {
  bone1Name=neck_01
  bone2Name=spine_05
  swing1LimitDegrees=45
  swing1Motion=Free
  swing2LimitDegrees=45
  swing2Motion=Free
  twistLimitDegrees=45
  twistMotion=Free
 }
 {
  bone1Name=neck_02
  bone2Name=neck_01
  swing1LimitDegrees=45
  swing1Motion=Free
  swing2LimitDegrees=45
  swing2Motion=Free
  twistLimitDegrees=45
  twistMotion=Free
 }
 {
  bone1Name=head
  bone2Name=neck_02
  swing1LimitDegrees=45
  swing1Motion=Free
  swing2LimitDegrees=45
  swing2Motion=Free
  twistLimitDegrees=45
  twistMotion=Free
 }
 {
  bone1Name=base_ponytail
  bone2Name=head
  swing1LimitDegrees=45
  swing1Motion=Free
  swing2LimitDegrees=45
  swing2Motion=Free
  twistLimitDegrees=45
  twistMotion=Free
 }
 {bone1Name=dyn_ponytailM_01, bone2Name=base_ponytail, swing1LimitDegrees=35, swing2LimitDegrees=15, twistLimitDegrees=0}
 {bone1Name=dyn_ponytailM_03, bone2Name=dyn_ponytailM_01, swing1LimitDegrees=35, swing2LimitDegrees=35, twistLimitDegrees=0}
 {bone1Name=dyn_ponytailM_05, bone2Name=dyn_ponytailM_04, swing1LimitDegrees=35, swing2LimitDegrees=35, twistLimitDegrees=0}
 {bone1Name=dyn_ponytailM_07, bone2Name=dyn_ponytailM_06, swing1LimitDegrees=35, swing2LimitDegrees=35, twistLimitDegrees=0}
 {bone1Name=dyn_ponytailM_08, bone2Name=dyn_ponytailM_07, swing1LimitDegrees=35, swing2LimitDegrees=35, twistLimitDegrees=0}
 {bone1Name=dyn_bang_01, bone2Name=base_hair, swing1LimitDegrees=15, swing2LimitDegrees=25, twistLimitDegrees=0}
 {bone1Name=dyn_bang_02, bone2Name=dyn_bang_01, swing1LimitDegrees=15, swing2LimitDegrees=25, twistLimitDegrees=0}
 {bone1Name=dyn_bang_03, bone2Name=dyn_bang_02, swing1LimitDegrees=15, swing2LimitDegrees=25, twistLimitDegrees=0}
 {bone1Name=dyn_bang_04, bone2Name=dyn_bang_03, swing1LimitDegrees=15, swing2LimitDegrees=25, twistLimitDegrees=0}
 {bone1Name=dyn_bang_05, bone2Name=dyn_bang_04, swing1LimitDegrees=15, swing2LimitDegrees=25, twistLimitDegrees=0}
 {bone1Name=dyn_bang_06, bone2Name=dyn_bang_02, swing1LimitDegrees=15, swing2LimitDegrees=25, twistLimitDegrees=0}
 {bone1Name=dyn_bang_07, bone2Name=dyn_bang_06, swing1LimitDegrees=15, swing2LimitDegrees=25, twistLimitDegrees=0}
]
