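# Post-process style AnimBP for the Shire female cosmetic's face accessory: hair, ponytail and collar follow the body with secondary motion.
# Copied mesh pose runs through Control Rig and collar AnimDynamics, then feeds three switched head-space RigidBody variants (default, pack, cape).
# A front-end idle curve blends slot animation onto hair bones, before eased additive ponytail bone corrections.
# Event graph tracks gravity override, backbling, zipline and head-tilt flags.
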AnimBlueprint /BRCosmetics/Characters/Player/Female/Medium/Bodies/F_MED_SteedShimmer_Shire/Meshes/Parts/F_MED_SteedShimmer_Shire_FaceAcc_AnimBP
target_skeleton=&/F_MED_SteedShimmer_Shire_FaceAcc_Skeleton
graphs [
 AnimGraph {nodes [
   Local To Component()
   n1 = Layered blend per bone(
    BasePose=Use cached pose 'Disabling_Pose_FE'()
    BlendPoses_0=Slot 'DefaultSlot'()
    BlendWeights_0=Get Custom Idle Is Playing Float() -> (Custom Idle Is Playing Float:double)
   ) {layerSetup [{branchFilters [
       {boneName=base_hair}
       {boneName=base_ponytail}
       {boneName=dyn_ponytailM_01}
       {boneName=dyn_ponytailM_03}
       {boneName=dyn_ponytailM_04}
       {boneName=dyn_ponytailM_05}
       {boneName=dyn_ponytailM_06}
       {boneName=dyn_ponytailM_07}
       {boneName=dyn_ponytailM_08}
       {boneName=dyn_bang_01}
       {boneName=dyn_bang_02}
       {boneName=dyn_bang_03}
       {boneName=dyn_bang_04}
       {boneName=dyn_bang_05}
       {boneName=dyn_bang_06}
       {boneName=dyn_bang_07}
      ]}]}
   n2 = "Transform (Modify) Bone - Bone: dyn_ponytailM_04"!(ComponentPose=Local To Component(LocalPose=@n1), bAlphaBoolEnabled:bool=True, Translation="0,0,0", Rotation="0,-10,0", Scale="1,1,1") -> (Pose) {
    alphaInputType=Bool
    rotationMode=BMM_Additive
    rotationSpace=BCS_BoneSpace
    alphaBoolBlend {blendInTime=0.8, blendOutTime=0.8}
    boneToModify {boneName=dyn_ponytailM_04}
    rotation {yaw=-10}
   }
   n3 = "Transform (Modify) Bone - Bone: dyn_ponytailM_05"!(ComponentPose=@n2, bAlphaBoolEnabled:bool=True, Translation="0,0,0", Rotation="0,0,-40", Scale="1,1,1") -> (Pose) {
    alphaInputType=Bool
    rotationMode=BMM_Additive
    rotationSpace=BCS_BoneSpace
    alphaBoolBlend {blendInTime=0.8, blendOutTime=0.8}
    boneToModify {boneName=dyn_ponytailM_05}
    rotation {roll=-40}
   }
   n4 = "Transform (Modify) Bone - Bone: base_ponytail"!(ComponentPose=@n3, bAlphaBoolEnabled:bool=True, Translation="0,0,0", Rotation="20,0,0", Scale="1,1,1") -> (Pose) {
    alphaInputType=Bool
    rotationMode=BMM_Additive
    rotationSpace=BCS_BoneSpace
    scaleMode=BMM_Replace
    scaleSpace=BCS_BoneSpace
    alphaBoolBlend {blendInTime=0.8, blendOutTime=0.8}
    boneToModify {boneName=base_ponytail}
    rotation {pitch=20}
   }
   n5 = "Transform (Modify) Bone - Bone: dyn_ponytailM_03"!(ComponentPose=@n4, bAlphaBoolEnabled:bool=True, Translation="0,0,0", Rotation="-20,0,0", Scale="1,0.95,1") -> (Pose) {
    alphaInputType=Bool
    rotationMode=BMM_Additive
    rotationSpace=BCS_BoneSpace
    scaleMode=BMM_Replace
    scaleSpace=BCS_BoneSpace
    alphaBoolBlend {blendInTime=0.8, blendOutTime=0.8}
    boneToModify {boneName=dyn_ponytailM_03}
    rotation {pitch=-20}
    scale {y=0.95}
   }
   n6 = "Transform (Modify) Bone - Bone: dyn_ponytailM_05"(ComponentPose=@n5, bAlphaBoolEnabled=True, Translation="0,0,0", Rotation="-10,0,0", Scale="1,0.9,0.95") {
    alphaInputType=Bool
    rotationMode=BMM_Additive
    rotationSpace=BCS_BoneSpace
    scaleMode=BMM_Replace
    scaleSpace=BCS_BoneSpace
    alphaBoolBlend {blendInTime=0.8, blendOutTime=0.8}
    boneToModify {boneName=dyn_ponytailM_05}
    rotation {pitch=-10}
    scale {y=0.9, z=0.95}
   }
   n7 = "Transform (Modify) Bone - Bone: dyn_ponytailM_06"!(ComponentPose=@n6, bAlphaBoolEnabled:bool=True, Translation="0,0,0", Rotation="5,0,8", Scale="0.95,0.85,0.8") -> (Pose) {
    alphaInputType=Bool
    rotationMode=BMM_Additive
    rotationSpace=BCS_BoneSpace
    scaleMode=BMM_Replace
    scaleSpace=BCS_BoneSpace
    alphaBoolBlend {blendInTime=0.8, blendOutTime=0.8}
    boneToModify {boneName=dyn_ponytailM_06}
    rotation {pitch=5, roll=8}
    scale=<0.95 0.85 0.8>
   }
   n8 = "Transform (Modify) Bone - Bone: dyn_ponytailM_07"!(ComponentPose=@n7, bAlphaBoolEnabled:bool=True, Translation="0,0,0", Rotation="0,0,10", Scale="0.9,0.9,0.95") -> (Pose) {
    alphaInputType=Bool
    rotationMode=BMM_Additive
    rotationSpace=BCS_BoneSpace
    scaleMode=BMM_Replace
    scaleSpace=BCS_BoneSpace
    alphaBoolBlend {blendInTime=0.8, blendOutTime=0.8}
    boneToModify {boneName=dyn_ponytailM_07}
    rotation {roll=10}
    scale=<0.9 0.9 0.95>
   }
   n9 = "Transform (Modify) Bone - Bone: dyn_ponytailM_04"(ComponentPose=@n8, bAlphaBoolEnabled=True, Translation="0,0,0", Rotation="10,-90,0", Scale="1,1,1") {
    alphaInputType=Bool
    rotationMode=BMM_Additive
    rotationSpace=BCS_BoneSpace
    alphaBoolBlend {blendInTime=0.8, blendOutTime=0.8}
    boneToModify {boneName=dyn_ponytailM_04}
    rotation {pitch=10, yaw=-90}
   }
   Output Pose(Result=Component To Local(ComponentPose=@n9)) {layerGroup=None}
   n10 = Control Rig!(Source=Copy Pose From Mesh(SourceMeshComponent=Get SkelMesh()) {bUseAttachedParent=true}, Alpha, NoRBAN:bool=false) -> (Pose) {controlRigAssetReference {blueprintRigClass=&/F_MED_SteedShimmer_Shire_FaceAcc_CtrlRig.F_MED_SteedShimmer_Shire_FaceAcc_CtrlRig_C}}
   n11 = Anim Dynamics - Bone: dyn_collar!(ComponentPose=Local To Component(LocalPose=@n10), Alpha=1, GravityOverride:Vector="0,0,-980", ExternalForce="0,0,0") -> (Pose) {
    angularDampingOverride=1
    angularSpringConstant=20000
    bAngularSpring=true
    bOverrideAngularDamping=true
    bUseGravityOverride=true
    gravityScale=0.35
    simulationSpace=RootRelative
    boundBone {boneName=dyn_collar}
    componentLinearAccScale=<1 1 1>
    componentLinearVelScale=<1 1 1>
    gravityOverride {z=-980}
    physicsBodyDefinitions [{
      boundBone {boneName=dyn_collar}
      boxExtents=<20 30 30>
      constraintSetup {angularLimitsMax=<0 0 7>, angularLimitsMin=<0 -10 -7>, angularTarget=<1 0 0>}
      localJointOffset=<6 0 0>
     }]
   }
   n12 = "Transform (Modify) Bone - Bone: base_collar"!(ComponentPose=@n11, bAlphaBoolEnabled:bool=True, Translation="0,0,0", Rotation="10,0,0", Scale="1,1,1") -> (Pose) {alphaInputType=Bool, rotationMode=BMM_Additive, rotationSpace=BCS_BoneSpace, boneToModify {boneName=base_collar}, rotation {pitch=10}}
   Save cached pose 'CR_AnimDynam'(Pose=Component To Local(ComponentPose=@n12)) {cacheName=CR_AnimDynam}
   n13 = RigidBody!(ComponentPose=Local To Component(LocalPose=Use cached pose 'CR_AnimDynam'()), Alpha, OverrideWorldGravity:Vector="0,0,-980", ExternalForce) -> (Pose) {
    bClampLinearTranslationLimitToRefPose=true
    bOverrideWorldGravity=true
    overridePhysicsAsset=&/F_MED_SteedShimmer_Shire_FaceAcc_Pack_Physics
    simulationSpace=BaseBoneSpace
    alphaScaleBiasClamp {bClampResult=true}
    baseBoneRef {boneName=head}
    componentAppliedLinearAccClamp=<500 500 500>
    componentLinearAccScale=<0.4 0.3 0.3>
    componentLinearVelScale=<0.4 0.3 0.3>
    overrideWorldGravity {z=-980}
    simSpaceSettings {worldAlpha=0.25}
   }
   n14 = Component To Local(ComponentPose=@n13)
   n15 = RigidBody(ComponentPose=Local To Component(LocalPose=Use cached pose 'CR_AnimDynam'()), OverrideWorldGravity="0,0,-980") {
    bClampLinearTranslationLimitToRefPose=true
    bOverrideWorldGravity=true
    overridePhysicsAsset=&/F_MED_SteedShimmer_Shire_FaceAcc_Physics
    simulationSpace=BaseBoneSpace
    alphaScaleBiasClamp {bClampResult=true}
    baseBoneRef {boneName=head}
    componentAppliedLinearAccClamp=<500 500 500>
    componentLinearAccScale=<0.4 0.3 0.3>
    componentLinearVelScale=<0.4 0.3 0.3>
    overrideWorldGravity {z=-980}
    simSpaceSettings {worldAlpha=0.25}
   }
   n16 = Component To Local(ComponentPose=@n15)
   n17 = RigidBody(ComponentPose=Local To Component(LocalPose=Use cached pose 'CR_AnimDynam'()), OverrideWorldGravity="0,0,-980") {
    bClampLinearTranslationLimitToRefPose=true
    bOverrideWorldGravity=true
    overridePhysicsAsset=&/F_MED_SteedShimmer_Shire_FaceAcc_Cape_Physics
    simulationSpace=BaseBoneSpace
    alphaScaleBiasClamp {bClampResult=true}
    baseBoneRef {boneName=head}
    componentAppliedLinearAccClamp=<500 500 500>
    componentLinearAccScale=<0.4 0.3 0.3>
    componentLinearVelScale=<0.4 0.3 0.3>
    overrideWorldGravity {z=-980}
    simSpaceSettings {worldAlpha=0.25}
   }
   Save cached pose 'Disabling_Pose_FE'(Pose=Blend Poses by bool(BlendPose_0=Component To Local(ComponentPose=@n17), BlendPose_1=Blend Poses by bool(BlendPose_0=@n14, BlendPose_1=@n16) {blendType=Linear}) {blendType=Linear}) {cacheName=Disabling_Pose_FE}
  ]}
 EventGraph {nodes [
   Event BlueprintUpdateAnimation()
   > n1 = "Gravity Override State Function (Parameter Asset) (Nativized)"(
    CharacterPartAnimInstance=Self-Reference()
    GameTimeInSeconds=GetGameTimeInSeconds()
    Parameters=&/F_MED_SteedShimmer_Shire_FA_Params
   )
   > Set Gravity Override(Gravity Override=@n1)
   > Set NoRBANEnabled(NoRBANEnabled:bool=NOT Boolean(A=Get bRigidBodyEnabled())) -> (Output_Get:bool)
   > Set toggleAnimDynamics(toggleAnimDynamics:bool=NOT Boolean(A=Get isOnZipline() -> (isOnZipline:bool))) -> (Output_Get:bool)
   > Branch(Condition=Get bIsInFrontEnd())
   > Set Custom Idle Is Playing Float(Custom Idle Is Playing Float:double=GetCurveValue(self=GetOwnerFrontEndAnimInstance(), CurveName=Frontend_Idle_Bypass)) -> (Output_Get:double)
   n2 = Get Is Wearing BP() -> (Is Wearing BP:bool)
   n3 = GetOwnerSkeletalMeshComponent()
   n4 = BreakRotator(InRot=BreakTransform(InTransform=GetDeltaTransformFromRefPose(self=@n3, BoneName=head, BaseName=spine_01)).Rotation)
   n5 = MapRangeClamped(Value=@n4, InRangeA=0, InRangeB=-30, OutRangeB=35)
   > Set head_rotate(head_rotate:bool=AND Boolean(A=NOT Boolean(A=@n2), B="float > float"(A=@n5, B=34))) -> (Output_Get:bool)
   > Set head_rotate_BP(head_rotate_BP:bool=AND Boolean!(A="float > float"(A=@n5, B=25), B="float < float"(A=@n4.Yaw), C:bool=Get Is Wearing BP()) -> (ReturnValue)) -> (Output_Get:bool)
   Event OnCharacterCustomizationLoaded()
   n6 = CheckForBackbling()
   > n7 = Set Is Wearing BP(Is Wearing BP:bool=OR Boolean!(A=@n6, B=@n6.bIsBackblingACape, C:bool=@n6.bIsBackblingLong) -> (ReturnValue)) -> (Output_Get:bool)
   > n8 = Set Is Wearing Cape(Is Wearing Cape=@n6.bIsBackblingLong)
   > Set Is Backbling Equipped(Is Backbling Equipped:bool=OR Boolean!(A=@n8, B=@n7) -> (ReturnValue)) -> (Output_Get:bool)
   Event BlueprintInitializeAnimation()
   > Set SkelMesh(SkelMesh=GetOwnerSkeletalMeshComponent())
   n9 = Event BlueprintProcessAnimInputEvent()
   > n10 = Branch(Condition=IsNamed(self=@n9.AnimInputEvent, InEventName=aie.Player.Zipline.Entered))
   > Set isOnZipline(isOnZipline:bool=1) -> (Output_Get:bool)
   Branch(execute=@n10.else, Condition=IsNamed(self=@n9.AnimInputEvent, InEventName=aie.Player.Zipline.Exited))
   > Set isOnZipline(isOnZipline=false)
  ]}
]
variables [
 Custom Idle Is Playing Float {type=number}
 Gravity Override {type=Vector}
 head_rotate {type=boolean}
 head_rotate_BP {type=boolean}
 Is Backbling Equipped {type=boolean}
 Is Wearing BP {type=boolean}
 Is Wearing Cape {type=boolean}
 isOnZipline {type=boolean}
 NoRBANEnabled {type=boolean}
 SkelMesh {type=SkeletalMeshComponent}
 toggleAnimDynamics {type=boolean}
]
